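# Whisker material for the ChubbyJingle large male player body, lit and usable on skeletal meshes with morph targets and mesh deformers.
# SplineThicken drives World Position Offset from base and tip width parameters, with color blended between tip and root by a UV mask.
# A flattened tube normal map, non-tangent-space, switches to its one-minus version by two-sided sign.
# The attributes feed MF_CharacterParent_2023_PostFX alongside pre-skinned position, distance blend and Fresnel.
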Material /BRCosmetics/Characters/Player/Male/Large/Bodies/M_LRG_ChubbyJingle/Materials/M_Whisker_ChubbyJingle
graph {
 properties {
  bTangentSpaceNormal=false
  bUseMaterialAttributes=true
  bUsedWithMeshDeformer=true
  bUsedWithMorphTargets=true
  bUsedWithSkeletalMesh=true
  pixelDepthOffsetMode=PDOM_Legacy
  usedShadingModels=Default Lit
 }
 nodes [
  n1 = TextureCoordinate()
  n2 = MaterialFunctionCall(
   Normal=TextureSample() {samplerType=SAMPLERTYPE_Normal, texture=/Game/Animation/Props/Kite/Textures/Tube_Normal_Tube_N}.RGB
   Flatness=ScalarParameter() {defaultValue=-1.5, group=SplineThicken, parameterName=ST_TubeNormalFlatness, sortPriority=33}
  ) {materialFunction=&FlattenNormal}
  n3 = TwoSidedSign()
  n4 = SetMaterialAttributes!(
   MaterialAttributes
   Base Color=LinearInterpolate(
    A=VectorParameter() {group=SplineThicken, parameterName=ST_BaseTipcolor, defaultValue=<rgba 0.5 0.5 0.5 1>}.RGB
    B=VectorParameter() {group=SplineThicken, parameterName=ST_BaseRootcolor, defaultValue=<rgba 0.5 0.5 0.5 1>}.RGB
    Alpha=StaticComponentMaskParameter(None=@n1) {defaultG=true}
   )
   Normal=LinearInterpolate(A=OneMinus(None=@n2.Result), B=@n2.Result, Alpha=Saturate(None=@n3))
   World Position Offset=MaterialFunctionCall(
    WidthBase=ScalarParameter() {defaultValue=0.15, group=SplineThicken, parameterName=ST_WidthBase}
    WidthTip=ScalarParameter() {defaultValue=0.05, group=SplineThicken, parameterName=ST_WidthTip}
    AngleCorrectedNormals=StaticBoolParameter() {group=SplineThicken, parameterName=ST_AngleCorrectedNormals}
   ) {materialFunction=&SplineThicken}.WorldPositionOffset
   Metallic=ScalarParameter() {group=SplineThicken, parameterName=ST_Metallic}
   Specular=ScalarParameter() {defaultValue=0.5, group=SplineThicken, parameterName=ST_Specular}
   Roughness=ScalarParameter() {defaultValue=0.5, group=SplineThicken, parameterName=ST_Roughness}
  ) {attributeSetTypes [
    #1
    #2
    #3
    #4
    #5
    #6
   ]}
  n5 = LocalPosition() {localOrigin=InstancePreSkinning}
  n6 = Constant() {r=1}
  MaterialFunctionCall(
   Material=@n4
   M=Constant4Vector().RGBA
   Pre Skinned Local Position=ShaderStageSwitch(PixelShader=VertexInterpolator(VS=@n5.XYZ).PS, VertexShader=@n5.XYZ)
   Distance Blend=MaterialFunctionCall!(Blend Range=Constant() {desc=4500, r=3000}, Start Offset=Constant() {r=-500}) {materialFunction=/Engine/Functions/Engine_MaterialFunctions03/Particles/Distance_Blend}.Result
   Fresnel=NamedRerouteDeclaration(None=Fresnel(ExponentIn=@n6) {baseReflectFraction=0}) {name=Fresnel, nodeColor=<rgba 1 0.3047 0 1>}
  ) {materialFunction=&MF_CharacterParent_2023_PostFX}
 ]
}
