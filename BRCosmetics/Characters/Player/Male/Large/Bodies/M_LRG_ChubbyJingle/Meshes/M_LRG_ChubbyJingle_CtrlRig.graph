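# Control Rig for the M_LRG_ChubbyJingle large male player body, applying procedural arm and costume deformation fixes.
# Unless emoting or wearing gauntlets, it twists upper arms via spherical pose readers, then re-solves PBIK toward cached hand transforms.
# A shared DeformRig function precedes pose-reader and constraint correctives for waistbelt, foot, armpit and neck bones.
# A right upper-arm reader also drives the LeftArmScarfHide curve, zeroed during emotes.
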ControlRigBlueprint /BRCosmetics/Characters/Player/Male/Large/Bodies/M_LRG_ChubbyJingle/Meshes/M_LRG_ChubbyJingle_CtrlRig
graphs [
 New Function_ContainedGraph {nodes [
   n1 = RigVMFunctionEntryNode()
   > RigVMFunction_Sequence_2 = RigVMAggregateNode()
   RigUnit_SetTransform = &RigUnit_SetTransform(
    Weight=1
    bInitial=false
    Value=&RigUnit_GetTransform(Item=(Type=Bone,Name="hand_l"), bInitial=false).Transform
    ExecutePin=@RigVMFunction_Sequence_2.A
    Item=@n1.Left Arm Controller
   )
   > &RigUnit_SetTransform(
    Weight=1
    bInitial=false
    Value=&RigUnit_GetTransform(Item=(Type=Bone,Name="hand_r"), bInitial=false).Transform
    Item=@n1.Right Arm Controller
   )
   > ComputePoleVector = RigVMFunctionReferenceNode(
    Bone A=(Type=Bone,Name="upperarm_l")
    Bone B=(Type=Bone,Name="lowerarm_l")
    Bone C=(Type=Bone,Name="hand_l")
    Initial=false
    OffsetFactor=5
    Origin Scale=(X=0,Y=0,Z=0)
    Draw Transform=@n1.Draw Debug
   )
   > &RigUnit_SetTransform(Weight=1, bInitial=false, Value=@ComputePoleVector.Transform, Item=@n1.Left Arm Pole Vector)
   > ComputePoleVector_1 = RigVMFunctionReferenceNode(
    Bone A=(Type=Bone,Name="upperarm_r")
    Bone B=(Type=Bone,Name="lowerarm_r")
    Bone C=(Type=Bone,Name="hand_r")
    Initial=false
    OffsetFactor=5
    Origin Scale=(X=0,Y=0,Z=0)
    Draw Transform=@n1.Draw Debug
   )
   > &RigUnit_SetTransform(Weight=1, bInitial=false, Value=@ComputePoleVector_1.Transform, Item=@n1.Right Arm Pole Vector)
   > SetControlVisibility = /Script/ControlRig.RigUnit_SetControlVisibility(bVisible=@n1.Draw Debug, Item=@n1.Left Arm Controller) {is=struct RigVMUnitNode, template="SetControlVisibility::Execute(in Item,in Pattern,in bVisible)"}
   > &RigUnit_SetControlVisibility(bVisible=@n1.Draw Debug, Item=@n1.Left Arm Pole Vector) {template="SetControlVisibility::Execute(in Item,in Pattern,in bVisible)"}
   > &RigUnit_SetControlVisibility(bVisible=@n1.Draw Debug, Item=@n1.Right Arm Controller) {template="SetControlVisibility::Execute(in Item,in Pattern,in bVisible)"}
   > &RigUnit_SetControlVisibility(bVisible=@n1.Draw Debug, Item=@n1.Right Arm Pole Vector) {template="SetControlVisibility::Execute(in Item,in Pattern,in bVisible)"}
   n2 = &RigUnit_SphericalPoseReader(
    ActiveRegionSize=0.1
    Debug=(bDrawDebug=true,bDraw2D=false,bDrawLocalAxes=true,DebugScale=25,DebugSegments=20,DebugThickness=0.25)
    DriverAxis=(X=1,Y=0,Z=0)
    DriverItem=(Type=Bone,Name="upperarm_l")
    FalloffSize=0.2
    FlipHeightScaling=False
    FlipWidthScaling=False
    OptionalParentItem=(Type=Bone,Name="spine_03")
    RotationOffset=(X=0,Y=-145,Z=0)
    ExecutePin=@RigVMFunction_Sequence_2.C
    FalloffRegionScaleFactors=@n1.Falloff Region Scale Factors
    Debug.bDrawDebug=@n1.Draw Debug
    ActiveRegionScaleFactors=@n1.Active Region Scale Factors
   )
   > ApplyFBIKChainCorrection = RigVMFunctionReferenceNode(End=(Type=Bone,Name="hand_l"), QuatAmount=-0.8, Start=(Type=Bone,Name="upperarm_l"), Weight=@n2.OutputParam)
   > n3 = &RigUnit_SphericalPoseReader(
    ActiveRegionSize=0.1
    Debug=(bDrawDebug=true,bDraw2D=true,bDrawLocalAxes=true,DebugScale=25,DebugSegments=20,DebugThickness=0.25)
    DriverAxis=(X=-1,Y=0,Z=0)
    DriverItem=(Type=Bone,Name="upperarm_r")
    FalloffSize=0.2
    FlipHeightScaling=False
    FlipWidthScaling=False
    OptionalParentItem=(Type=Bone,Name="spine_03")
    RotationOffset=(X=0,Y=35,Z=0)
    FalloffRegionScaleFactors=@n1.Falloff Region Scale Factors
    Debug.bDraw2D=@n1.Draw Debug
    Debug.bDrawDebug=@n1.Draw Debug
    Debug.bDrawLocalAxes=@n1.Draw Debug
    ActiveRegionScaleFactors=@n1.Active Region Scale Factors
   )
   > &RigUnit_ModifyTransforms(
    ItemToModify=((Item=(Type=Control,Name="Arm_R_PV"),Transform=(Rotation=(X=0,Y=0,Z=0,W=-1),Translation=(X=0,Y=66.756,Z=0),Scale3D=(X=1,Y=1,Z=1))))
    WeightMinimum=0
    Weight=@n3.OutputParam
    ItemToModify.0.Item=@n1.Right Arm Pole Vector
    ItemToModify.0.Transform.Translation=@n1.Right PV Add
   )
   /Script/PBIK.RigUnit_PBIK(
    BoneSettings=((Bone="lowerarm_l",RotationStiffness=0,PositionStiffness=0,X=Locked,MinX=0,MaxX=0,Y=Locked,MinY=0,MaxY=0,Z=Free,MinZ=0,MaxZ=180,bUsePreferredAngles=true,PreferredAngles=(X=0,Y=0,Z=1)))
    Debug=(DrawScale=1,bDrawDebug=False)
    Effectors="((Bone=\"ik_hand_l\",Transform=(Rotation=(X=0,Y=0,Z=0,W=1),Translation=(X=0,Y=0,Z=0),Scale3D=(X=1,Y=1,Z=1)),PositionAlpha=1,RotationAlpha=1,StrengthAlpha=1,PullChainAlpha=1,PinRotation=1,ChainDepth=0),(Bone=\"upperarm_l\",Transform=(Rotation=(X=0,Y=0,Z=0,W=1), ...(truncated)"
    ExcludedBones=()
    Root=upperarm_l
    Settings="(Iterations=20,SubIterations=0,MassMultiplier=1,bAllowStretch=false,RootBehavior=PrePull,PrePullRootSettings=(RotationAlpha=0,RotationAlphaX=1,RotationAlphaY=1,RotationAlphaZ=1,PositionAlpha=1,PositionAlphaX=1,PositionAlphaY=1,PositionAlphaZ=1),GlobalPullC ...(truncated)"
    ExecutePin=@RigVMFunction_Sequence_2.D
   ) {is=struct RigVMUnitNode}
   > /Script/ControlRig.RigUnit_TwoBoneIKSimplePerItem(
    DebugSettings=(bEnabled=False,Scale=10,WorldOffset=(Rotation=(X=0,Y=0,Z=0,W=1),Translation=(X=0,Y=0,Z=0),Scale3D=(X=1,Y=1,Z=1)))
    EffectorItem=(Type=Bone,Name="hand_r")
    ItemA=(Type=Bone,Name="upperarm_r")
    ItemALength=0
    ItemB=(Type=Bone,Name="lowerarm_r")
    ItemBLength=0
    PoleVectorKind=Location
    PoleVectorSpace=(Type=Bone,Name="None")
    PrimaryAxis=(X=-1,Y=0,Z=0)
    SecondaryAxis=(X=0,Y=1,Z=0)
    SecondaryAxisWeight=1
    StretchMaximumRatio=1.25
    StretchStartRatio=1
    Weight=1
    bEnableStretch=false
    bPropagateToChildren=True
    Effector=&RigUnit_GetTransform(bInitial=False, Item=@n1.Right Arm Controller).Transform
    PoleVector=&RigUnit_GetTransform(bInitial=False, Item=@n1.Right Arm Pole Vector).Transform.Translation
   ) {is=struct RigVMUnitNode}
   RigVMFunctionReturnNode(Output Param=@n3.OutputParam, Value=@n2.OutputParam, ExecuteContext=@RigVMFunction_Sequence_2.E)
   n4 = &RigUnit_SphericalPoseReader(
    ActiveRegionScaleFactors=(PositiveWidth=1,NegativeWidth=1,PositiveHeight=1,NegativeHeight=1)
    ActiveRegionSize=0.1
    Debug=(bDrawDebug=True,bDraw2D=False,bDrawLocalAxes=False,DebugScale=25,DebugSegments=20,DebugThickness=0.25)
    DriverAxis=(X=1,Y=0,Z=0)
    DriverItem=(Type=Bone,Name="lowerarm_l")
    ExecutePin=()
    FalloffRegionScaleFactors=(PositiveWidth=1,NegativeWidth=1,PositiveHeight=1.8,NegativeHeight=1)
    FalloffSize=0.2
    FlipHeightScaling=False
    FlipWidthScaling=False
    OptionalParentItem=(Type=Bone,Name="None")
    RotationOffset=(X=-15.6637,Y=-88.487,Z=-35)
   )
   > RigVMVariableNode(Value=@n4.OutputParam) {variable=SphericalPoseReader_OutputParam}
   &RigUnit_ModifyTransforms(
    ExecutePin=()
    ItemToModify="((Item=(Type=Control,Name=\"Arm_L_PV\"),Transform=(Rotation=(X=0,Y=0,Z=0,W=1),Translation=(X=0,Y=-123.5555,Z=0),Scale3D=(X=1,Y=1,Z=1))),(Item=(Type=Control,Name=\"Arm_L_IK\"),Transform=(Rotation=(X=0,Y=0,Z=0,W=1),Translation=(X=0,Y=0,Z=0),Scale3D=(X=1,Y=1,Z=1) ...(truncated)"
    Weight=1
    WeightMinimum=0
    ItemToModify.0.Item=@n1.Left Arm Pole Vector
    ItemToModify.1.Item=@n1.Left Arm Controller
    ItemToModify.0.Transform.Translation=@n1.Left PV Add
   )
   &RigUnit_ModifyTransforms(
    ExecutePin=()
    ItemToModify=((Item=(Type=Control,Name="L_Arm_PV"),Transform=(Rotation=(X=0,Y=0,Z=0,W=1),Translation=(X=0,Y=-63.9814,Z=10.2044),Scale3D=(X=1,Y=1,Z=1))))
    WeightMinimum=0
    Weight=@n4.OutputParam
   )
   RigVMRerouteNode(Value=@n1.Left Arm Pole Vector)
   RigVMRerouteNode(Value=@RigVMFunction_Sequence_2.B)
   &RigUnit_TwoBoneIKSimplePerItem(
    DebugSettings=(bEnabled=False,Scale=10,WorldOffset=(Rotation=(X=0,Y=0,Z=0,W=1),Translation=(X=0,Y=0,Z=0),Scale3D=(X=1,Y=1,Z=1)))
    EffectorItem=(Type=Bone,Name="hand_l")
    ExecutePin=()
    ItemA=(Type=Bone,Name="upperarm_l")
    ItemALength=0
    ItemB=(Type=Bone,Name="lowerarm_l")
    ItemBLength=0
    PoleVectorKind=Location
    PoleVectorSpace=(Type=Bone,Name="None")
    PrimaryAxis=(X=1,Y=0,Z=0)
    SecondaryAxis=(X=0,Y=-1,Z=0)
    SecondaryAxisWeight=1
    StretchMaximumRatio=1.25
    StretchStartRatio=1
    Weight=1
    bEnableStretch=false
    bPropagateToChildren=True
    Effector=&RigUnit_GetTransform(bInitial=False, Item=@n1.Left Arm Controller).Transform
    PoleVector=&RigUnit_GetTransform(bInitial=False, Item=@n1.Left Arm Pole Vector).Transform.Translation
   )
  ]}
 New Function_ContainedGraph {nodes [
   n1 = RigVMFunctionEntryNode()
   n2 = &RigUnit_GetTransform(bInitial=False, Item=@n1.Start)
   Subtract = /Script/RigVM.RigVMFunction_MathVectorSub(B=@n2.Transform.Translation, A=&RigUnit_GetTransform(bInitial=False, Item=@n1.End).Transform.Translation) {is=struct RigVMUnitNode, template="Subtract::Execute(in A,in B,out Result)"}
   Unit = /Script/RigVM.RigVMFunction_MathVectorUnit(Value=@Subtract.Result) {is=struct RigVMUnitNode, template="Unit::Execute(in Value,out Result)"}
   RigVMFunction_MathQuaternionFromAxisAndAngle = /Script/RigVM.RigVMFunction_MathQuaternionFromAxisAndAngle(Axis=@Unit.Result, Angle=@n1.QuatAmount) {is=struct RigVMUnitNode, template="FromAxisAndAngle::Execute(in Axis,in Angle,out Result)"}
   Multiply = /Script/RigVM.RigVMFunction_MathQuaternionMul(B=@n2.Transform.Rotation, A=@RigVMFunction_MathQuaternionFromAxisAndAngle.Result) {is=struct RigVMUnitNode, template="Multiply::Execute(in A,in B,out Result)"}
   > &RigUnit_SetRotation(Space=GlobalSpace, bInitial=False, Item=@n1.Start, Weight=@n1.Weight, Value=@Multiply.Result)
   > RigVMFunctionReturnNode()
  ]}
 New Function_ContainedGraph {nodes [
   n1 = RigVMFunctionEntryNode()
   > RigVMFunction_Sequence = RigVMAggregateNode()
   n2 = &RigUnit_SphericalPoseReader(
    ActiveRegionScaleFactors=(PositiveWidth=2,NegativeWidth=0.5,PositiveHeight=1,NegativeHeight=1)
    ActiveRegionSize=0.1
    Debug=(bDrawDebug=true,bDraw2D=false,bDrawLocalAxes=false,DebugScale=25,DebugSegments=20,DebugThickness=0.25)
    DriverAxis=(X=1,Y=0,Z=0)
    DriverItem=(Type=Bone,Name="upperarm_l")
    FalloffRegionScaleFactors=(PositiveWidth=1,NegativeWidth=1,PositiveHeight=2.5,NegativeHeight=2.5)
    FalloffSize=0.2
    FlipHeightScaling=False
    FlipWidthScaling=False
    OptionalParentItem=(Type=Bone,Name="spine_03")
    RotationOffset=(X=0,Y=-145,Z=0)
    ExecutePin=@RigVMFunction_Sequence.B
    Debug.bDrawDebug=@n1.Draw Debug
   )
   > ApplyFBIKChainCorrection = RigVMFunctionReferenceNode(End=(Type=Bone,Name="hand_l"), Start=(Type=Bone,Name="upperarm_l"), Weight=@n2.OutputParam, QuatAmount=@n1.Left Arm Rad)
   > n3 = &RigUnit_SphericalPoseReader(
    ActiveRegionScaleFactors=(PositiveWidth=2,NegativeWidth=0.5,PositiveHeight=1,NegativeHeight=1)
    ActiveRegionSize=0.1
    Debug=(bDrawDebug=true,bDraw2D=false,bDrawLocalAxes=false,DebugScale=25,DebugSegments=20,DebugThickness=0.25)
    DriverAxis=(X=-1,Y=0,Z=0)
    DriverItem=(Type=Bone,Name="upperarm_r")
    FalloffRegionScaleFactors=(PositiveWidth=2,NegativeWidth=1,PositiveHeight=2,NegativeHeight=2)
    FalloffSize=0.2
    FlipHeightScaling=False
    FlipWidthScaling=False
    OptionalParentItem=(Type=Bone,Name="spine_03")
    RotationOffset=(X=0,Y=35,Z=0)
    Debug.bDrawDebug=@n1.Draw Debug
   )
   > ApplyFBIKChainCorrection_1 = RigVMFunctionReferenceNode(End=(Type=Bone,Name="hand_r"), Start=(Type=Bone,Name="upperarm_r"), Weight=@n3.OutputParam, QuatAmount=@n1.Right Arm Rad)
   &RigUnit_ParentConstraint(
    AdvancedSettings=(InterpolationType=Average,RotationOrderForFilter=XZY)
    Child=(Type=Bone,Name="hand_l")
    Parents=((Item=(Type=Bone,Name="None"),Weight=1))
    Weight=1
    bMaintainOffset=false
    ExecutePin=@RigVMFunction_Sequence.C
    Parents.0.Item.Name=@n1.Left Arm Ctrl Bone
   )
   > &RigUnit_ParentConstraint(
    AdvancedSettings=(InterpolationType=Average,RotationOrderForFilter=XZY)
    Child=(Type=Bone,Name="hand_r")
    Parents=((Item=(Type=Bone,Name="None"),Weight=1))
    Weight=1
    bMaintainOffset=false
    Parents.0.Item.Name=@n1.Right Arm Ctrl Bone
   )
   > &RigUnit_PBIK(
    BoneSettings="((Bone=\"lowerarm_l\",RotationStiffness=0,PositionStiffness=0,X=Locked,MinX=0,MaxX=0,Y=Locked,MinY=0,MaxY=0,Z=Limited,MinZ=0,MaxZ=180,bUsePreferredAngles=true,PreferredAngles=(X=0,Y=0,Z=1)),(Bone=\"hand_l\",RotationStiffness=0,PositionStiffness=0,X=Free,MinX=0 ...(truncated)"
    Debug=(DrawScale=1,bDrawDebug=False)
    Effectors="((Bone=\"l_arm_ctrl\",Transform=(Rotation=(X=0,Y=0,Z=0,W=1),Translation=(X=0,Y=0,Z=0),Scale3D=(X=1,Y=1,Z=1)),PositionAlpha=1,RotationAlpha=1,StrengthAlpha=1,PullChainAlpha=1,PinRotation=1,ChainDepth=0),(Bone=\"upperarm_l\",Transform=(Rotation=(X=0,Y=0,Z=0,W=1) ...(truncated)"
    ExcludedBones=()
    Root=upperarm_l
    Settings="(Iterations=20,SubIterations=0,MassMultiplier=1,bAllowStretch=false,RootBehavior=PrePull,PrePullRootSettings=(RotationAlpha=0,RotationAlphaX=1,RotationAlphaY=1,RotationAlphaZ=1,PositionAlpha=1,PositionAlphaX=1,PositionAlphaY=1,PositionAlphaZ=1),GlobalPullC ...(truncated)"
    Effectors.0.Bone=@n1.Left Arm Ctrl Bone
   )
   > &RigUnit_PBIK(
    BoneSettings=((Bone="lowerarm_r",RotationStiffness=0,PositionStiffness=0,X=Locked,MinX=0,MaxX=0,Y=Locked,MinY=0,MaxY=0,Z=Free,MinZ=0,MaxZ=180,bUsePreferredAngles=true,PreferredAngles=(X=0,Y=0,Z=1)))
    Debug=(DrawScale=1,bDrawDebug=False)
    Effectors="((Bone=\"r_arm_ctrl\",Transform=(Rotation=(X=0,Y=0,Z=0,W=1),Translation=(X=0,Y=0,Z=0),Scale3D=(X=1,Y=1,Z=1)),PositionAlpha=1,RotationAlpha=1,StrengthAlpha=1,PullChainAlpha=1,PinRotation=1,ChainDepth=0),(Bone=\"upperarm_r\",Transform=(Rotation=(X=0,Y=0,Z=0,W=1) ...(truncated)"
    ExcludedBones=()
    Root=upperarm_r
    Settings="(Iterations=20,SubIterations=0,MassMultiplier=1,bAllowStretch=false,RootBehavior=PrePull,PrePullRootSettings=(RotationAlpha=0,RotationAlphaX=1,RotationAlphaY=1,RotationAlphaZ=1,PositionAlpha=1,PositionAlphaX=1,PositionAlphaY=1,PositionAlphaZ=1),GlobalPullC ...(truncated)"
    Effectors.0.Bone=@n1.Right Arm Ctrl Bone
   )
   RigUnit_GetTransform = &RigUnit_GetTransform(Item=(Type=Bone,Name="hand_l"), bInitial=False)
   RigUnit_SetTransform = &RigUnit_SetTransform(
    Item=(Type=Bone,Name="l_arm_ctrl")
    Weight=1
    bInitial=False
    Value=@RigUnit_GetTransform.Transform
    ExecutePin=@RigVMFunction_Sequence.A
    Item.Name=@n1.Left Arm Ctrl Bone
   )
   > &RigUnit_SetTransform(
    Item=(Type=Bone,Name="r_arm_ctrl")
    Weight=1
    bInitial=False
    Value=&RigUnit_GetTransform(Item=(Type=Bone,Name="hand_r"), bInitial=False).Transform
    Item.Name=@n1.Right Arm Ctrl Bone
   )
   RigVMFunctionReturnNode(ExecuteContext=@RigVMFunction_Sequence.D)
  ]}
 RigVMFunctionLibrary {nodes [
   Adjust IK Elbows for M_LRG v2 = RigVMCollapseNode(
    Active Region Scale Factors=(PositiveWidth=2,NegativeWidth=0.5,PositiveHeight=1,NegativeHeight=1)
    Draw Debug=true
    ExecuteContext=()
    Falloff Region Scale Factors=(PositiveWidth=2,NegativeWidth=1,PositiveHeight=2,NegativeHeight=2)
    Left Arm Controller=(Type=Control,Name="Arm_L_IK")
    Left Arm Pole Vector=(Type=Control,Name="Arm_R_IK")
    Left PV Add=(X=0,Y=-123.5555,Z=0)
    Right Arm Controller=(Type=Control,Name="Arm_R_IK")
    Right Arm Pole Vector=(Type=Control,Name="Arm_R_PV")
    Right PV Add=(X=0,Y=66.756,Z=0)
   )
   Adjust IK Elbows for M_LRG v3 = RigVMCollapseNode(Draw Debug=true, ExecuteContext=(), Left Arm Ctrl Bone=l_arm_ctrl, Left Arm Rad=-0.8, Right Arm Ctrl Bone=r_arm_ctrl, Right Arm Rad=-0.8)
   ApplyFBIKChainCorrection = RigVMCollapseNode(End=(Type=None,Name="None"), ExecuteContext=(), Start=(Type=None,Name="None"))
  ]}
 RigVMFunction_Sequence_1_ContainedGraph {nodes [
   n1 = RigVMFunctionEntryNode()
   RigVMFunction_Sequence = &RigVMFunction_Sequence(ExecuteContext=@n1.ExecuteContext)
   n2 = &RigVMFunction_Sequence(ExecuteContext=@RigVMFunction_Sequence.B)
   n3 = &RigVMFunction_Sequence(ExecuteContext=@n2.B)
   RigVMFunctionReturnNode(A=@RigVMFunction_Sequence.A, B=@n2.A, C=@n3.A, D=@n3.B)
  ]}
 RigVMFunction_Sequence_3_ContainedGraph {nodes [
   n1 = RigVMFunctionEntryNode()
   n2 = &RigVMFunction_Sequence(ExecuteContext=@n1.ExecuteContext)
   n3 = &RigVMFunction_Sequence(ExecuteContext=@n2.B)
   n4 = &RigVMFunction_Sequence(ExecuteContext=@n3.B)
   n5 = &RigVMFunction_Sequence(ExecuteContext=@n4.B)
   n6 = &RigVMFunction_Sequence(ExecuteContext=@n5.B)
   RigVMFunctionReturnNode(B=@n3.A, A=@n2.A, C=@n4.A, D=@n5.A, E=@n6.A, F=@n6.B)
  ]}
 RigVMFunction_Sequence_4_ContainedGraph {nodes [
   n1 = RigVMFunctionEntryNode()
   n2 = &RigVMFunction_Sequence(ExecuteContext=@n1.ExecuteContext)
   n3 = &RigVMFunction_Sequence(ExecuteContext=@n2.B)
   RigVMFunctionReturnNode(B=@n3.A, C=@n3.B, A=@n2.A)
  ]}
 RigVMFunction_Sequence_4_ContainedGraph {nodes [
   n1 = RigVMFunctionEntryNode()
   n2 = &RigVMFunction_Sequence(ExecuteContext=@n1.ExecuteContext)
   n3 = &RigVMFunction_Sequence(ExecuteContext=@n2.B)
   n4 = &RigVMFunction_Sequence(ExecuteContext=@n3.B)
   RigVMFunctionReturnNode(A=@n2.A, B=@n3.A, C=@n4.A, D=@n4.B)
  ]}
 RigVMFunction_Sequence_5_ContainedGraph {nodes [
   n1 = RigVMFunctionEntryNode()
   n2 = &RigVMFunction_Sequence(ExecuteContext=@n1.ExecuteContext)
   n3 = &RigVMFunction_Sequence(ExecuteContext=@n2.B)
   RigVMFunctionReturnNode(B=@n3.A, C=@n3.B, A=@n2.A)
  ]}
 RigVMModel {nodes [
   RigUnit_BeginExecution = &RigUnit_BeginExecution()
   n1 = &RigVMFunction_Sequence(ExecuteContext=@RigUnit_BeginExecution.ExecutePin)
   &RigUnit_ModifyTransforms(
    ItemToModify=((Item=(Type=Bone,Name="upperarm_l"),Transform=(Rotation=(X=0,Y=0,Z=0,W=1),Translation=(X=0,Y=0,Z=0),Scale3D=(X=1,Y=1,Z=1))))
    Weight=1
    WeightMinimum=0
    ExecutePin=@n1.A
   )
   RigVMFunction_Sequence = &RigVMFunction_Sequence(ExecuteContext=@n1.B)
   RigVMFunction_MathBoolOr = &RigVMFunction_MathBoolOr(B=RigVMVariableNode() {variable=isGauntletsEquip}.Value, A=RigVMVariableNode() {variable=isPlayingEmote}.Value)
   RigVMFunction_ControlFlowBranch = &RigVMFunction_ControlFlowBranch(Condition=@RigVMFunction_MathBoolOr.Result, ExecuteContext=@RigVMFunction_Sequence.A)
   Adjust IK Elbows for M_LRG v3 = RigVMFunctionReferenceNode(
    Draw Debug=false
    Left Arm Ctrl Bone=l_arm_ctrl
    Left Arm Rad=-0.4
    Right Arm Ctrl Bone=r_arm_ctrl
    Right Arm Rad=0.4
    ExecuteContext=@RigVMFunction_ControlFlowBranch.False
   )
   FN_DeformRig_FNC = RigVMFunctionReferenceNode(
    DeformRig_ Curve="(deformRig_l_upperarm_up_alpha_72_#1=False,deformRig_r_upperarm_up_alpha_73_#2=False,deformRig_l_upperarm_fr_alpha_74_#3=False,deformRig_r_upperarm_fr_alpha_75_CB51B8 ...(truncated)"
    DeformRig_Bool="(upperarm_1_#4=True,elbow_3_#5=True,wrist_7_#6=True,groin_8_#7=True,glute_17_#8=True,knee_9_40C408324CA261 ...(truncated)"
    DeformRig_Debug=(upperarm_debug_18_#4=False,wrist_debug_20_#6=False,lat_debug_26_#9=False,groinGlute_debug_27_#7=False)
    DeformRig_Value="(upperarm_up_18_#10=(Rotation=(X=0,Y=0,Z=0,W=1),Translation=(X=0,Y=0,Z=8),Scale3D=(X=0.7,Y=1,Z=1)),upperarm_fwd_fr_19_#11=(Rotation=(X=0,Y=0,Z=0,W=1),Translation=(X=0,Y=10,Z=0),Scale3D=(X=0.5,Y=1,Z= ...(truncated)"
    ExecuteContext=@RigVMFunction_Sequence.B
   )
   > RigVMFunction_Sequence_1 = RigVMAggregateNode()
   RigVMFunction_Sequence_2 = RigVMAggregateNode(ExecuteContext=@RigVMFunction_Sequence_1.A)
   n2 = &RigUnit_SphericalPoseReader(
    ActiveRegionScaleFactors=(PositiveWidth=3.5,NegativeWidth=0.5,PositiveHeight=3,NegativeHeight=3)
    ActiveRegionSize=0.1
    Debug=(bDrawDebug=false,DebugScale=25,DebugSegments=20,DebugThickness=0.25)
    DriverAxis=(X=-1,Y=0,Z=0)
    DriverItem=(Type=Bone,Name="thigh_l")
    FalloffRegionScaleFactors=(PositiveWidth=1,NegativeWidth=1,PositiveHeight=1,NegativeHeight=1)
    FalloffSize=0.45
    OptionalParentItem=(Type=Bone)
    RotationOffset=(X=124.2688,Y=-3.2809,Z=0)
    ExecutePin=@RigVMFunction_Sequence_2.A
   )
   > &RigUnit_ModifyTransforms(
    ItemToModify="((Item=(Type=Bone,Name=\"dyn_waistbelt_ml\"),Transform=(Rotation=(X=-0.1163,Y=0.2587,Z=-0.0243,W=-0.9586),Translation=(X=-4.6889,Y=-0.4037,Z=8.5729),Scale3D=(X=1,Y=1,Z=0.6293))),(Item=(Type=Bone,Name=\"dyn_waistbelt_l\"),Transform=(Rotation=(X=-0.1193,Y=0,Z=0, ...(truncated)"
    WeightMinimum=0
    Weight=@n2.OutputParam
   )
   n3 = &RigUnit_SphericalPoseReader(
    ActiveRegionScaleFactors=(PositiveWidth=3.5,NegativeWidth=0.5,PositiveHeight=3,NegativeHeight=3)
    ActiveRegionSize=0.1
    Debug=(bDrawDebug=false,DebugScale=25,DebugSegments=20,DebugThickness=0.25)
    DriverAxis=(X=1,Y=0,Z=0)
    DriverItem=(Type=Bone,Name="thigh_r")
    FalloffRegionScaleFactors=(PositiveWidth=1,NegativeWidth=1,PositiveHeight=1,NegativeHeight=1)
    FalloffSize=0.45
    FlipHeightScaling=false
    FlipWidthScaling=true
    OptionalParentItem=(Type=Bone)
    RotationOffset=(X=-56.214,Y=0,Z=0)
    ExecutePin=@RigVMFunction_Sequence_2.B
   )
   > &RigUnit_ModifyTransforms(
    ItemToModify="((Item=(Type=Bone,Name=\"dyn_waistbelt_mr\"),Transform=(Rotation=(X=-0.1486,Y=0.2363,Z=0.0366,W=0.9596),Translation=(X=-5.4548,Y=-1.0372,Z=-8.5105),Scale3D=(X=1,Y=1,Z=0.6293))),(Item=(Type=Bone,Name=\"dyn_waistbelt_r\"),Transform=(Rotation=(X=-0.1193,Y=0,Z=0,W ...(truncated)"
    WeightMinimum=0
    Weight=@n3.OutputParam
   )
   n4 = &RigUnit_SphericalPoseReader(
    ActiveRegionScaleFactors=(PositiveWidth=1,NegativeWidth=1,PositiveHeight=2,NegativeHeight=2)
    ActiveRegionSize=0.1
    Debug=(bDrawDebug=false,bDraw2D=False,bDrawLocalAxes=False,DebugScale=25,DebugSegments=20,DebugThickness=0.25)
    DriverAxis=(X=1,Y=0,Z=0)
    DriverItem=(Type=Bone,Name="foot_l")
    FalloffRegionScaleFactors=(PositiveWidth=1,NegativeWidth=1,PositiveHeight=1,NegativeHeight=1)
    FalloffSize=0.4
    FlipHeightScaling=False
    FlipWidthScaling=False
    OptionalParentItem=(Type=Bone,Name="None")
    RotationOffset=(X=90,Y=-0.2908,Z=-8.4086)
    ExecutePin=@RigVMFunction_Sequence_1.B
   )
   > &RigUnit_ModifyTransforms(
    ItemToModify=((Item=(Type=Bone,Name="dyn_foot_bk_l"),Transform=(Rotation=(X=0,Y=0.5,Z=0,W=0.866),Translation=(X=11.5815,Y=0.0002,Z=-6.8904),Scale3D=(X=1,Y=1,Z=0.1))))
    WeightMinimum=0
    Weight=@n4.OutputParam
   )
   > n5 = &RigUnit_SphericalPoseReader(
    ActiveRegionScaleFactors=(PositiveWidth=1,NegativeWidth=1,PositiveHeight=2,NegativeHeight=2)
    ActiveRegionSize=0.1
    Debug=(bDrawDebug=false,bDraw2D=False,bDrawLocalAxes=False,DebugScale=25,DebugSegments=20,DebugThickness=0.25)
    DriverAxis=(X=1,Y=0,Z=0)
    DriverItem=(Type=Bone,Name="foot_r")
    FalloffRegionScaleFactors=(PositiveWidth=1,NegativeWidth=1,PositiveHeight=1,NegativeHeight=1)
    FalloffSize=0.4
    FlipHeightScaling=False
    FlipWidthScaling=False
    OptionalParentItem=(Type=Bone,Name="None")
    RotationOffset=(X=90,Y=-0.2908,Z=-8.4086)
   )
   > &RigUnit_ModifyTransforms(
    ItemToModify=((Item=(Type=Bone,Name="dyn_foot_bk_r"),Transform=(Rotation=(X=0,Y=0.5,Z=0,W=0.866),Translation=(X=11.5815,Y=0.0002,Z=-6.8904),Scale3D=(X=1,Y=1,Z=0.1))))
    WeightMinimum=0
    Weight=@n5.OutputParam
   )
   > n6 = &RigUnit_SphericalPoseReader(
    ActiveRegionScaleFactors=(PositiveWidth=1,NegativeWidth=1,PositiveHeight=2,NegativeHeight=2)
    ActiveRegionSize=0.1
    Debug=(bDrawDebug=False,bDraw2D=False,bDrawLocalAxes=False,DebugScale=25,DebugSegments=20,DebugThickness=0.25)
    DriverAxis=(X=-1,Y=0,Z=0)
    DriverItem=(Type=Bone,Name="foot_l")
    FalloffRegionScaleFactors=(PositiveWidth=0.8,NegativeWidth=0.8,PositiveHeight=0.8,NegativeHeight=0.8)
    FalloffSize=0.4
    FlipHeightScaling=False
    FlipWidthScaling=False
    OptionalParentItem=(Type=Bone,Name="None")
    RotationOffset=(X=90,Y=0,Z=28)
   )
   > &RigUnit_ModifyTransforms(
    ItemToModify=((Item=(Type=Bone,Name="dyn_foot_bk_l"),Transform=(Rotation=(X=0,Y=-0.587,Z=0,W=0.8096),Translation=(X=-7.8619,Y=0,Z=0.0119),Scale3D=(X=4.134,Y=1,Z=1.8522))))
    WeightMinimum=0
    Weight=@n6.OutputParam
   )
   > n7 = &RigUnit_SphericalPoseReader(
    ActiveRegionScaleFactors=(PositiveWidth=1,NegativeWidth=1,PositiveHeight=2,NegativeHeight=2)
    ActiveRegionSize=0.1
    Debug=(bDrawDebug=False,bDraw2D=False,bDrawLocalAxes=False,DebugScale=25,DebugSegments=20,DebugThickness=0.25)
    DriverAxis=(X=-1,Y=0,Z=0)
    DriverItem=(Type=Bone,Name="foot_r")
    FalloffRegionScaleFactors=(PositiveWidth=0.8,NegativeWidth=0.8,PositiveHeight=0.8,NegativeHeight=0.8)
    FalloffSize=0.4
    FlipHeightScaling=False
    FlipWidthScaling=False
    OptionalParentItem=(Type=Bone,Name="None")
    RotationOffset=(X=90,Y=0,Z=28)
   )
   > &RigUnit_ModifyTransforms(
    ItemToModify=((Item=(Type=Bone,Name="dyn_foot_bk_r"),Transform=(Rotation=(X=0,Y=-0.587,Z=0,W=0.8096),Translation=(X=-7.8619,Y=0,Z=0.0119),Scale3D=(X=4.134,Y=1,Z=1.8522))))
    WeightMinimum=0
    Weight=@n7.OutputParam
   )
   RigVMFunction_Sequence_3 = RigVMAggregateNode(ExecuteContext=@RigVMFunction_Sequence_1.D)
   /Script/ControlRig.RigUnit_PositionConstraintLocalSpaceOffset(
    Child=(Type=Bone,Name="dyn_armpit_l")
    Filter=(bX=True,bY=True,bZ=True)
    Parents=((Item=(Type=Bone,Name="upperarm_l"),Weight=0.6),(Item=(Type=Bone,Name="spine_05"),Weight=0.4))
    Weight=1
    bMaintainOffset=true
    ExecutePin=@RigVMFunction_Sequence_3.A
   ) {is=struct RigVMUnitNode}
   > n8 = &RigUnit_SphericalPoseReader(
    ActiveRegionScaleFactors=(PositiveWidth=1,NegativeWidth=3,PositiveHeight=2,NegativeHeight=2)
    ActiveRegionSize=0.1
    Debug=(bDrawDebug=True,DebugScale=25,DebugSegments=20,DebugThickness=0.25)
    DriverAxis=(X=1,Y=0,Z=0)
    DriverItem=(Type=Bone,Name="upperarm_l")
    FalloffRegionScaleFactors=(PositiveWidth=1,NegativeWidth=1,PositiveHeight=1,NegativeHeight=1)
    FalloffSize=0.4
    OptionalParentItem=(Type=Bone)
    RotationOffset=(X=90,Y=0,Z=-10)
   )
   > &RigUnit_ModifyTransforms(
    ItemToModify=((Item=(Type=Bone,Name="dyn_armpit_l"),Transform=(Rotation=(X=0.342,Y=0,Z=0,W=0.9397),Translation=(X=8,Y=8,Z=5),Scale3D=(X=1,Y=1,Z=1))))
    WeightMinimum=0
    Weight=@n8.OutputParam
   )
   > n9 = &RigUnit_SphericalPoseReader(
    ActiveRegionScaleFactors=(PositiveWidth=2,NegativeWidth=0.5,PositiveHeight=1.5,NegativeHeight=1.5)
    ActiveRegionSize=0.1
    Debug=(bDrawDebug=True,bDraw2D=false,bDrawLocalAxes=false,DebugScale=25,DebugSegments=20,DebugThickness=0.25)
    DriverAxis=(X=1,Y=0,Z=0)
    DriverItem=(Type=Bone,Name="upperarm_l")
    FalloffRegionScaleFactors=(PositiveWidth=1,NegativeWidth=1.6,PositiveHeight=3,NegativeHeight=2.5)
    FalloffSize=0.2
    FlipHeightScaling=False
    FlipWidthScaling=False
    OptionalParentItem=(Type=Bone,Name="spine_03")
    RotationOffset=(X=0,Y=-145,Z=0)
   )
   > &RigUnit_ModifyTransforms(
    ItemToModify=((Item=(Type=Bone,Name="dyn_armpit_l"),Transform=(Rotation=(X=0,Y=-0.1483,Z=0,W=-0.9889),Translation=(X=1.358,Y=0,Z=0.1048),Scale3D=(X=1,Y=1,Z=1))))
    WeightMinimum=0
    Weight=@n9.OutputParam
   )
   > n10 = &RigUnit_SphericalPoseReader(
    ActiveRegionScaleFactors=(PositiveWidth=2.5,NegativeWidth=3,PositiveHeight=3,NegativeHeight=3)
    ActiveRegionSize=0.1
    Debug=(bDrawDebug=True,DebugScale=25,DebugSegments=20,DebugThickness=0.25)
    DriverAxis=(X=1,Y=0,Z=0)
    DriverItem=(Type=Bone,Name="upperarm_l")
    FalloffRegionScaleFactors=(PositiveWidth=1,NegativeWidth=1,PositiveHeight=1,NegativeHeight=1)
    FalloffSize=0.5
    OptionalParentItem=(Type=Bone)
    RotationOffset=(X=0,Y=0,Z=0)
   )
   > &RigUnit_ModifyTransforms(
    ItemToModify=((Item=(Type=Bone,Name="dyn_armpit_l"),Transform=(Rotation=(X=0,Y=0.2588,Z=0,W=-0.9659),Translation=(X=2.3965,Y=0,Z=5.844),Scale3D=(X=1,Y=1,Z=1))))
    WeightMinimum=0
    Weight=@n10.OutputParam
   )
   &RigUnit_PositionConstraintLocalSpaceOffset(
    Child=(Type=Bone,Name="dyn_armpit_r")
    Filter=(bX=True,bY=True,bZ=True)
    Parents=((Item=(Type=Bone,Name="upperarm_r"),Weight=0.6),(Item=(Type=Bone,Name="spine_05"),Weight=0.4))
    Weight=1
    bMaintainOffset=True
    ExecutePin=@RigVMFunction_Sequence_3.B
   )
   > n11 = &RigUnit_SphericalPoseReader(
    ActiveRegionScaleFactors=(PositiveWidth=0.5,NegativeWidth=3,PositiveHeight=1.5,NegativeHeight=1.5)
    ActiveRegionSize=0.1
    Debug=(bDrawDebug=True,DebugScale=25,DebugSegments=20,DebugThickness=0.25)
    DriverAxis=(X=-1,Y=0,Z=0)
    DriverItem=(Type=Bone,Name="upperarm_r")
    FalloffRegionScaleFactors=(PositiveWidth=1,NegativeWidth=1,PositiveHeight=1,NegativeHeight=1)
    FalloffSize=0.4
    FlipWidthScaling=true
    OptionalParentItem=(Type=Bone)
    RotationOffset=(X=90,Y=180,Z=180)
   )
   > &RigUnit_ModifyTransforms(
    ItemToModify=((Item=(Type=Bone,Name="dyn_armpit_r"),Transform=(Rotation=(X=-0.342,Y=0,Z=0,W=0.9397),Translation=(X=6,Y=-8,Z=5),Scale3D=(X=1,Y=1,Z=1))))
    WeightMinimum=0
    Weight=@n11.OutputParam
   )
   > n12 = &RigUnit_SphericalPoseReader(
    ActiveRegionScaleFactors=(PositiveWidth=2,NegativeWidth=0.5,PositiveHeight=1,NegativeHeight=1)
    ActiveRegionSize=0.1
    Debug=(bDrawDebug=true,bDraw2D=false,bDrawLocalAxes=false,DebugScale=25,DebugSegments=20,DebugThickness=0.25)
    DriverAxis=(X=-1,Y=0,Z=0)
    DriverItem=(Type=Bone,Name="upperarm_r")
    FalloffRegionScaleFactors=(PositiveWidth=2,NegativeWidth=1.6,PositiveHeight=2,NegativeHeight=2)
    FalloffSize=0.2
    FlipHeightScaling=False
    FlipWidthScaling=False
    OptionalParentItem=(Type=Bone,Name="spine_03")
    RotationOffset=(X=0,Y=35,Z=0)
   )
   > &RigUnit_ModifyTransforms(
    ItemToModify=((Item=(Type=Bone,Name="dyn_armpit_r"),Transform=(Rotation=(X=0,Y=0.1483,Z=0,W=0.9889),Translation=(X=1.358,Y=0,Z=0.1048),Scale3D=(X=1,Y=1,Z=1))))
    WeightMinimum=0
    Weight=@n12.OutputParam
   )
   > n13 = &RigUnit_SphericalPoseReader(
    ActiveRegionScaleFactors=(PositiveWidth=2.5,NegativeWidth=3,PositiveHeight=3,NegativeHeight=3)
    ActiveRegionSize=0.1
    Debug=(bDrawDebug=false,DebugScale=25,DebugSegments=20,DebugThickness=0.25)
    DriverAxis=(X=-1,Y=0,Z=0)
    DriverItem=(Type=Bone,Name="upperarm_r")
    FalloffRegionScaleFactors=(PositiveWidth=1,NegativeWidth=1,PositiveHeight=1,NegativeHeight=1)
    FalloffSize=0.5
    FlipHeightScaling=true
    OptionalParentItem=(Type=Bone)
    RotationOffset=(X=180,Y=0,Z=180)
   )
   > &RigUnit_ModifyTransforms(
    ItemToModify=((Item=(Type=Bone,Name="dyn_armpit_r"),Transform=(Rotation=(X=0,Y=0,Z=0,W=1),Translation=(X=5,Y=0,Z=15),Scale3D=(X=1,Y=1,Z=1))))
    WeightMinimum=0
    Weight=@n13.OutputParam
   )
   RigVMFunction_MathBoolToInteger = &RigVMFunction_MathBoolToInteger(Value=RigVMVariableNode() {variable=isPlayingEmote}.Value)
   /Script/ControlRig.RigUnit_SetCurveValue(
    Curve=LeftArmScarfHide
    Value=RigVMDispatchNode(Values=(0,0), Index=@RigVMFunction_MathBoolToInteger.Result, Values.0=@n11.OutputParam) {template="DISPATCH_RigVMDispatch_SelectInt32(in Index,in Values,out Result)"}.Result
    ExecutePin=@RigVMFunction_Sequence_3.C
   ) {is=struct RigVMUnitNode}
   &RigUnit_ParentConstraint(
    AdvancedSettings=(InterpolationType=Average,RotationOrderForFilter=XZY)
    Child=(Type=Bone,Name="dyn_neck_fr_base")
    Parents=((Item=(Type=Bone,Name="head"),Weight=0.3),(Item=(Type=Bone,Name="spine_05"),Weight=0.7))
    Weight=1
    ExecutePin=@RigVMFunction_Sequence_1.C
   )
   > &RigUnit_ParentConstraint(
    AdvancedSettings=(InterpolationType=Average,RotationOrderForFilter=XZY)
    Child=(Type=Bone,Name="dyn_neck_bk_base")
    Parents=((Item=(Type=Bone,Name="head"),Weight=0.3),(Item=(Type=Bone,Name="spine_05"),Weight=0.7))
    Weight=1
   )
   > n14 = &RigUnit_SphericalPoseReader(
    ActiveRegionScaleFactors=(PositiveWidth=1.1,NegativeWidth=1.1,PositiveHeight=1.1,NegativeHeight=1.1)
    ActiveRegionSize=0.1
    Debug=(bDrawDebug=false,bDraw2D=False,bDrawLocalAxes=False,DebugScale=25,DebugSegments=20,DebugThickness=0.25)
    DriverAxis=(X=1,Y=0,Z=0)
    DriverItem=(Type=Bone,Name="clavicle_l")
    FalloffRegionScaleFactors=(PositiveWidth=1.4,NegativeWidth=1.4,PositiveHeight=1.4,NegativeHeight=1.4)
    FalloffSize=0.2
    FlipHeightScaling=False
    FlipWidthScaling=False
    OptionalParentItem=(Type=Bone,Name="None")
    RotationOffset=(X=0,Y=-45,Z=0)
   )
   > &RigUnit_ModifyTransforms(
    ItemToModify=((Item=(Type=Bone,Name="dyn_neck_l"),Transform=(Rotation=(X=0,Y=-0.1964,Z=0,W=0.9805),Translation=(X=0.2497,Y=0,Z=-0.2822),Scale3D=(X=1,Y=1,Z=0.5697))))
    WeightMinimum=0
    Weight=@n14.OutputParam
   )
   > n15 = &RigUnit_SphericalPoseReader(
    ActiveRegionScaleFactors=(PositiveWidth=1.1,NegativeWidth=1.1,PositiveHeight=1.1,NegativeHeight=1.1)
    ActiveRegionSize=0.1
    Debug=(bDrawDebug=false,bDraw2D=False,bDrawLocalAxes=False,DebugScale=25,DebugSegments=20,DebugThickness=0.25)
    DriverAxis=(X=1,Y=0,Z=0)
    DriverItem=(Type=Bone,Name="clavicle_r")
    FalloffRegionScaleFactors=(PositiveWidth=1.4,NegativeWidth=1.4,PositiveHeight=1.4,NegativeHeight=1.4)
    FalloffSize=0.2
    FlipHeightScaling=False
    FlipWidthScaling=False
    OptionalParentItem=(Type=Bone,Name="None")
    RotationOffset=(X=0,Y=-45,Z=0)
   )
   > &RigUnit_ModifyTransforms(
    ItemToModify=((Item=(Type=Bone,Name="dyn_neck_r"),Transform=(Rotation=(X=0,Y=0.1964,Z=0,W=0.9805),Translation=(X=-0.2497,Y=0,Z=0.2822),Scale3D=(X=1,Y=1,Z=0.5697))))
    WeightMinimum=0
    Weight=@n15.OutputParam
   )
   n16 = &RigUnit_SphericalPoseReader(
    ActiveRegionScaleFactors=(PositiveWidth=1,NegativeWidth=1,PositiveHeight=1,NegativeHeight=1)
    ActiveRegionSize=0.05
    Debug=(bDrawDebug=True,bDraw2D=False,bDrawLocalAxes=False,DebugScale=25,DebugSegments=20,DebugThickness=0.25)
    DriverAxis=(X=1,Y=0,Z=0)
    DriverItem=(Type=Bone,Name="upperarm_l")
    ExecutePin=()
    FalloffRegionScaleFactors=(PositiveWidth=1,NegativeWidth=1,PositiveHeight=1,NegativeHeight=1)
    FalloffSize=0.18
    FlipHeightScaling=False
    FlipWidthScaling=False
    OptionalParentItem=(Type=Bone,Name="None")
    RotationOffset=(X=0,Y=-55.9271,Z=0)
   )
   > &RigUnit_ModifyTransforms(
    ItemToModify=((Item=(Type=Bone,Name="dyn_armpit_l"),Transform=(Rotation=(X=0,Y=0,Z=0,W=-1),Translation=(X=-1.9782,Y=0,Z=-11.0496),Scale3D=(X=1,Y=1,Z=1))))
    WeightMinimum=0
    Weight=@n16.OutputParam
   )
   n17 = &RigUnit_SphericalPoseReader(
    ActiveRegionScaleFactors=(PositiveWidth=1,NegativeWidth=1,PositiveHeight=1,NegativeHeight=1)
    ActiveRegionSize=0.05
    Debug=(bDrawDebug=True,bDraw2D=False,bDrawLocalAxes=False,DebugScale=25,DebugSegments=20,DebugThickness=0.25)
    DriverAxis=(X=1,Y=0,Z=0)
    DriverItem=(Type=Bone,Name="upperarm_r")
    ExecutePin=()
    FalloffRegionScaleFactors=(PositiveWidth=1,NegativeWidth=1,PositiveHeight=1,NegativeHeight=1)
    FalloffSize=0.18
    FlipHeightScaling=False
    FlipWidthScaling=False
    OptionalParentItem=(Type=Bone,Name="None")
    RotationOffset=(X=0,Y=-55.9271,Z=0)
   )
   > &RigUnit_ModifyTransforms(
    ItemToModify=((Item=(Type=Bone,Name="dyn_armpit_r"),Transform=(Rotation=(X=0,Y=0,Z=0,W=-1),Translation=(X=-1.9782,Y=0,Z=-11.0496),Scale3D=(X=1,Y=1,Z=1))))
    WeightMinimum=0
    Weight=@n17.OutputParam
   )
   &RigUnit_ParentConstraint(
    AdvancedSettings=(InterpolationType=Average,RotationOrderForFilter=XZY)
    Child=(Type=Bone,Name="dyn_neck_l_base")
    ExecutePin=()
    Parents=((Item=(Type=Bone,Name="head"),Weight=0.3),(Item=(Type=Bone,Name="clavicle_l"),Weight=0.7))
    Weight=1
   )
   &RigUnit_ParentConstraint(
    AdvancedSettings=(InterpolationType=Average,RotationOrderForFilter=XZY)
    Child=(Type=Bone,Name="dyn_neck_r_base")
    ExecutePin=()
    Parents=((Item=(Type=Bone,Name="head"),Weight=0.3),(Item=(Type=Bone,Name="clavicle_r"),Weight=0.7))
    Weight=1
   )
  ]}
]
hierarchy [
 root {loc=<0 0 0>, rot=<0 0 0 1>}
 pelvis {parent=root, loc=<0 -0.7321 92.4218>, rot=<0.0127 0.707 -0.0127 -0.707>}
 spine_01 {parent=pelvis, loc=<5.6002 -0.5128 0>, rot=<0 0 0.034 -0.9994>}
 spine_02 {parent=spine_01, loc=<8.0756 0.0139 0>, rot=<0 0 0.0238 -0.9997>}
 spine_03 {parent=spine_02, loc=<7.0395 0 0>, rot=<0 0 0.0149 -0.9999>}
 spine_04 {parent=spine_03, loc=<8.1023 0 0>, rot=<0 0 -0.0132 -0.9999>}
 spine_05 {parent=spine_04, loc=<10.0141 0 0>, rot=<0 0 -0.0015 -1>}
 clavicle_l {parent=spine_05, loc=<2.5856 -0.2866 -4.3777>, rot=<0.0465 -0.7289 -0.009 -0.6829>}
 upperarm_l {parent=clavicle_l, loc=<18.2264 0.1342 -0.001>, rot=<0.0253 -0.2693 0.0013 -0.9627>}
 lowerarm_l {parent=upperarm_l, loc=<29.0071 0.1686 -1.239>, rot=<0.0391 -0.0099 -0.2453 -0.9686>}
 hand_l {parent=lowerarm_l, loc=<25.8607 0.222 1.0274>, rot=<-0.7193 -0.0988 0.1193 -0.6773>}
 weapon_l {parent=hand_l, loc=<-0.1974 -3.4111 0.3811>, rot=<0 0 0 1>}
 hand_attach_l {parent=hand_l, loc=<8.5 -3 0>, rot=<0.7071 0.7071 0 0>}
 thumb_01_l {parent=hand_l, loc=<1.3118 -0.3191 -2.459>, rot=<0.6546 -0.4042 0.0664 -0.6354>}
 thumb_02_l {parent=thumb_01_l, loc=<5.9336 -0.0448 0.0401>, rot=<-0.0037 0.0213 0.298 -0.9543>}
 thumb_03_l {parent=thumb_02_l, loc=<2.2223 0.3682 -0.0907>, rot=<0.0118 -0.013 0.0066 -0.9998>}
 index_metacarpal_l {parent=hand_l, loc=<1.6535 0.0183 -2.96>, rot=<0.0158 -0.0124 -0.0173 -0.9996>}
 index_01_l {parent=index_metacarpal_l, loc=<8.3008 0.0357 -0.0524>, rot=<-0.0044 0.0485 0.0924 -0.9945>}
 index_02_l {parent=index_01_l, loc=<3.3587 -0.3004 0.0452>, rot=<-0.002 0.0027 0.3225 -0.9466>}
 index_03_l {parent=index_02_l, loc=<2.5578 -0.123 -0.0252>, rot=<-0.002 0.0067 0.2558 -0.9667>}
 middle_metacarpal_l {parent=hand_l, loc=<2.1856 -0.1725 -0.5984>, rot=<0.006 0.0168 -0.0329 -0.9993>}
 middle_01_l {parent=middle_metacarpal_l, loc=<7.3275 -0.2447 0.0129>, rot=<-0.0208 0.0042 0.0858 -0.9961>}
 middle_02_l {parent=middle_01_l, loc=<4.0123 0.1401 0.1175>, rot=<-0.0218 0.0258 0.3512 -0.9357>}
 middle_03_l {parent=middle_02_l, loc=<2.6077 0.0534 0.1373>, rot=<0.0082 0.0188 0.1708 -0.9851>}
 ring_metacarpal_l {parent=hand_l, loc=<2.2074 -0.3625 1.3638>, rot=<-0.0284 0.0722 -0.0458 -0.9959>}
 ring_01_l {parent=ring_metacarpal_l, loc=<7.0295 -0.6197 -0.1446>, rot=<-0.0489 -0.0252 0.1892 -0.9804>}
 ring_02_l {parent=ring_01_l, loc=<3.6174 0.4647 -0.1027>, rot=<0 0 0.2806 -0.9598>}
 ring_03_l {parent=ring_02_l, loc=<2.4152 0.114 0.0199>, rot=<0 0 0.1189 -0.9929>}
 pinky_metacarpal_l {parent=hand_l, loc=<1.7616 -0.9753 3.3814>, rot=<-0.0397 0.0814 -0.069 -0.9935>}
 pinky_01_l {parent=pinky_metacarpal_l, loc=<7.1709 -0.4831 -0.1094>, rot=<-0.0654 -0.0035 0.2242 -0.9723>}
 pinky_02_l {parent=pinky_01_l, loc=<2.8274 0.0893 -0.1043>, rot=<0.0037 0.0011 0.2854 -0.9584>}
 pinky_03_l {parent=pinky_02_l, loc=<2.2046 0.0307 -0.0078>, rot=<-0.0002 0.0098 0.1445 -0.9895>}
 deform_wrist_dn_base_l {parent=hand_l, loc=<0 -3.4173 0>, rot=<0 0 0 -1>}
 deform_wrist_dn_l {parent=deform_wrist_dn_base_l, loc=<0 0 0>, rot=<0 0 0 1>}
 deform_wrist_up_base_l {parent=hand_l, loc=<0 2.7805 0>, rot=<0 0 0 -1>}
 deform_wrist_up_l {parent=deform_wrist_up_base_l, loc=<0 0 0>, rot=<0 0 0 1>}
 lowerarm_twist_02_l {parent=lowerarm_l, loc=<9.6199 0 0>, rot=<0 0 0 -1>}
 lowerarm_twist_01_l {parent=lowerarm_l, loc=<17.8549 0 0>, rot=<0 0 0 -1>}
 deform_elbow_in_base_l {parent=lowerarm_l, loc=<0.2112 5.645 0.8274>, rot=<0 0 0 -1>}
 deform_elbow_in_l {parent=deform_elbow_in_base_l, loc=<0 0 0>, rot=<0 0 0 1>}
 deform_elbow_out_base_l {parent=lowerarm_l, loc=<0 -7.5129 0>, rot=<0 0 0 -1>}
 deform_elbow_out_l {parent=deform_elbow_out_base_l, loc=<0 0 0>, rot=<0 0 0 1>}
 upperarm_twist_01_l {parent=upperarm_l, loc=<7.9796 0.0843 -0.6195>, rot=<0 0 0 -1>}
 deform_upperarm_bk_base_l {parent=upperarm_twist_01_l, loc=<-6.8983 -13.3826 -1.3258>, rot=<0 0 0 1>}
 deform_upperarm_bk_l {parent=deform_upperarm_bk_base_l, loc=<0 0 0>, rot=<0 0 0 1>}
 deform_upperarm_fr_base_l {parent=upperarm_twist_01_l, loc=<-8.3487 10.7117 -2.1455>, rot=<0 0 0 1>}
 deform_upperarm_fr_l {parent=deform_upperarm_fr_base_l, loc=<0 0 0>, rot=<0 0 0 1>}
 deform_upperarm_up_base_l {parent=upperarm_twist_01_l, loc=<-11.2012 -1.2234 8.7784>, rot=<0 0 0 1>}
 deform_upperarm_up_l {parent=deform_upperarm_up_base_l, loc=<0 0 0>, rot=<0 0 0 1>}
 upperarm_twist_02_l {parent=upperarm_l, loc=<19.5059 0.0843 -0.6195>, rot=<0 0 0 -1>}
 deform_bicep_base_l {parent=upperarm_twist_02_l, loc=<0 8.0494 0>, rot=<0 0 0 1>}
 deform_bicep_l {parent=deform_bicep_base_l, loc=<0 0 0>, rot=<0 0 0 1>}
 neck_01 {parent=spine_05, loc=<10.815 1.9196 0>, rot=<0 0 -0.1864 -0.9825>}
 neck_02 {parent=neck_01, loc=<4.7347 0.3045 0>, rot=<0 0 0.0363 -0.9993>}
 head {parent=neck_02, loc=<5.7632 0.5552 0>, rot=<0 0 0.115 -0.9934>}
 dyn_glasses_base {parent=head, loc=<9.5896 -2.2145 0>, rot=<0.4979 0.5021 0.5021 0.4979>}
 dyn_glasses {parent=dyn_glasses_base, loc=<0 0 0>, rot=<0 0 0 1>}
 dyn_hair_f_base {parent=head, loc=<16.9726 -0.0589 8.7994>, rot=<-0.1443 -0.0751 -0.1407 -0.9766>}
 dyn_hair_f {parent=dyn_hair_f_base, loc=<0 0 0>, rot=<0 0 0 1>}
 dyn_hair_e_base {parent=head, loc=<15.5064 -8.342 5.1214>, rot=<-0.1443 -0.0751 -0.1407 -0.9766>}
 dyn_hair_e {parent=dyn_hair_e_base, loc=<0 0 0>, rot=<0 0 0 1>}
 dyn_hair_d_base {parent=head, loc=<18.102 -6.0025 -0.8513>, rot=<-0.1758 0.095 -0.1709 -0.9648>}
 dyn_hair_d {parent=dyn_hair_d_base, loc=<0 0 0>, rot=<0 0 0 1>}
 dyn_hair_c_base {parent=head, loc=<14.8682 -5.4765 -8.3693>, rot=<0.2838 -0.1786 -0.1057 -0.9361>}
 dyn_hair_c {parent=dyn_hair_c_base, loc=<0 0 0>, rot=<0 0 0 1>}
 dyn_hair_b_base {parent=head, loc=<17.6807 -0.5044 -5.5671>, rot=<-0.1498 0.2234 -0.1651 -0.9489>}
 dyn_hair_b {parent=dyn_hair_b_base, loc=<0 0 0>, rot=<0 0 0 1>}
 dyn_hair_a_base {parent=head, loc=<18.9836 2.971 2.7203>, rot=<-0.0264 -0.1015 -0.1094 -0.9885>}
 dyn_hair_a {parent=dyn_hair_a_base, loc=<0 0 0>, rot=<0 0 0 1>}
 dyn_hair_g_base {parent=head, loc=<15.945 -8.6121 -6.0646>, rot=<-0.1637 0.0102 0.0744 -0.9836>}
 dyn_hair_g {parent=dyn_hair_g_base, loc=<0 0 0>, rot=<0 0 0 1>}
 faceAttach {parent=head, loc=<1.938 -3.5964 0>, rot=<0.0029 -0.7071 -0.0029 -0.7071>}
 C_brow_mid {parent=faceAttach, loc=<0 12.6675 6.2603>, rot=<0 0 0 -1>}
 L_brow_mid {parent=faceAttach, loc=<3.6909 10.6935 7.9176>, rot=<-0.0001 -0.0042 0.2553 -0.9669>}
 L_brow_outer {parent=faceAttach, loc=<8.4929 7.3923 7.7671>, rot=<-0.0287 -0.04 0.4491 -0.8921>}
 L_cheek_inner {parent=faceAttach, loc=<7.8679 13.9803 -1.3444>, rot=<-0.0706 -0.2697 0.3184 -0.906>}
 C_lip_upper_mid {parent=faceAttach, loc=<-0.0106 17.0725 -4.2427>, rot=<0.2225 0.0013 -0.0005 -0.9749>}
 L_lip_upper_outer {parent=faceAttach, loc=<3.2603 16.1502 -4.5848>, rot=<0.0359 -0.0728 0.2627 -0.9615>}
 L_lip_corner {parent=faceAttach, loc=<5.8212 14.0616 -4.7832>, rot=<0.003 0.008 -0.0146 -0.9999>}
 C_nose_bridge {parent=faceAttach, loc=<0 17.9626 0.309>, rot=<-0.2218 0 0 -0.9751>}
 R_brow_mid {parent=faceAttach, loc=<-3.6909 10.6935 7.9176>, rot=<-0.0001 0.0042 -0.2553 -0.9669>}
 R_brow_outer {parent=faceAttach, loc=<-8.4929 7.3923 7.7671>, rot=<-0.0287 0.04 -0.4491 -0.8921>}
 R_cheek_inner {parent=faceAttach, loc=<-7.8679 13.9803 -1.3444>, rot=<-0.0706 0.2697 -0.3184 -0.906>}
 R_lip_corner {parent=faceAttach, loc=<-5.8212 14.0616 -4.7832>, rot=<0.003 -0.008 0.0146 -0.9999>}
 C_jaw {parent=faceAttach, loc=<-0.0134 6.8138 -3.0507>, rot=<0 0 0 1>}
 L_lip_lower_outer {parent=C_jaw, loc=<3.089 8.99 -2.2184>, rot=<0.013 0.0445 0.3057 -0.951>}
 C_lip_lower_mid {parent=C_jaw, loc=<-0.028 9.8799 -2.1375>, rot=<0 0 0 -1>}
 R_lip_lower_outer {parent=C_jaw, loc=<-3.0804 9.0038 -2.2152>, rot=<0.013 -0.0445 -0.3057 -0.951>}
 teeth_lower {parent=C_jaw, loc=<0.0134 4.579 -3.131>, rot=<0 0 0 1>}
 tongue {parent=C_jaw, loc=<0.0141 3.0976 -2.123>, rot=<-0.1568 0 0 -0.9876>}
 R_eye {parent=faceAttach, loc=<-5.3603 10.1738 3.2297>, rot=<0 0 0.9998 -0.0175>}
 R_eye_lid_upper_mid {parent=R_eye, loc=<0 0.0023 0.001>, rot=<0.0523 0 0 -0.9986>}
 R_eye_lid_lower_mid {parent=R_eye, loc=<0 0.0014 0.0007>, rot=<0.0329 0 0 -0.9995>}
 L_eye {parent=faceAttach, loc=<5.3142 10.1534 3.1874>, rot=<0 0 0.9998 0.0175>}
 L_eye_lid_upper_mid {parent=L_eye, loc=<-0.0227 -0.011 0.0212>, rot=<0.0523 0 0 -0.9986>}
 L_eye_lid_lower_mid {parent=L_eye, loc=<-0.0227 -0.011 0.0212>, rot=<0.0728 -0.0626 0.0024 -0.9954>}
 teeth_upper {parent=faceAttach, loc=<0 11.9734 -4.1298>, rot=<0 0 0 1>}
 R_lip_upper_outer {parent=faceAttach, loc=<-3.3391 16.1985 -4.5736>, rot=<0.0359 0.0728 -0.2627 -0.9615>}
 dyn_whisker_a_l_base {parent=head, loc=<1.1687 12.0239 -4.4485>, rot=<0.0684 -0.4871 -0.0387 -0.8698>}
 dyn_whisker_a_l {parent=dyn_whisker_a_l_base, loc=<0 0 0>, rot=<0 0 0 1>}
 dyn_whisker_b_l_base {parent=head, loc=<0.3594 12.5326 -4.3381>, rot=<-0.0724 -0.6317 -0.0648 -0.7691>}
 dyn_whisker_b_l {parent=dyn_whisker_b_l_base, loc=<0 0 0>, rot=<0 0 0 1>}
 dyn_whisker_c_l_base {parent=head, loc=<-0.2757 13.0875 -3.7953>, rot=<-0.1466 -0.6917 -0.1523 -0.6905>}
 dyn_whisker_c_l {parent=dyn_whisker_c_l_base, loc=<0 0 0>, rot=<0 0 0 1>}
 dyn_whisker_d_l_base {parent=head, loc=<-1.3638 12.1597 -5.1022>, rot=<-0.0539 -0.7237 -0.1249 -0.6765>}
 dyn_whisker_d_l {parent=dyn_whisker_d_l_base, loc=<0 0 0>, rot=<0 0 0 1>}
 dyn_whisker_d_r_base {parent=head, loc=<-0.7726 12.7909 4.3536>, rot=<0.6661 -0.0218 0.7393 -0.0959>}
 dyn_whisker_d_r {parent=dyn_whisker_d_r_base, loc=<0 0 0>, rot=<0 0 0 1>}
 dyn_whisker_c_r_base {parent=head, loc=<-0.237 12.5422 4.6041>, rot=<0.7001 0.0991 0.7009 0.0933>}
 dyn_whisker_c_r {parent=dyn_whisker_c_r_base, loc=<0 0 0>, rot=<0 0 0 1>}
 dyn_whisker_b_r_base {parent=head, loc=<0.5322 12.3713 4.4584>, rot=<0.7449 0.1057 0.6577 -0.0377>}
 dyn_whisker_b_r {parent=dyn_whisker_b_r_base, loc=<0 0 0>, rot=<0 0 0 1>}
 dyn_whisker_a_r_base {parent=head, loc=<1.2212 11.6876 4.7733>, rot=<0.7776 0.0237 0.6214 -0.0931>}
 dyn_whisker_a_r {parent=dyn_whisker_a_r_base, loc=<0 0 0>, rot=<0 0 0 1>}
 clavicle_r {parent=spine_05, loc=<2.5855 -0.2866 4.3777>, rot=<0.7289 0.0465 -0.6829 0.009>}
 upperarm_r {parent=clavicle_r, loc=<-18.2264 -0.1342 0.001>, rot=<0.0253 -0.2693 0.0013 -0.9627>}
 lowerarm_r {parent=upperarm_r, loc=<-29.0071 -0.1686 1.239>, rot=<0.0391 -0.0099 -0.2453 -0.9686>}
 hand_r {parent=lowerarm_r, loc=<-25.8607 -0.222 -1.0274>, rot=<-0.7193 -0.0988 0.1193 -0.6773>}
 hand_attach_r {parent=hand_r, loc=<-8.5 3 0>, rot=<0 0 -0.7071 -0.7071>}
 thumb_01_r {parent=hand_r, loc=<-1.3118 0.3191 2.459>, rot=<0.6546 -0.4042 0.0664 -0.6354>}
 thumb_02_r {parent=thumb_01_r, loc=<-5.9336 0.0448 -0.0401>, rot=<-0.0037 0.0213 0.298 -0.9543>}
 thumb_03_r {parent=thumb_02_r, loc=<-2.2223 -0.3682 0.0907>, rot=<0.0118 -0.013 0.0066 -0.9998>}
 index_metacarpal_r {parent=hand_r, loc=<-1.6535 -0.0183 2.96>, rot=<0.0158 -0.0124 -0.0173 -0.9996>}
 index_01_r {parent=index_metacarpal_r, loc=<-8.3008 -0.0357 0.0524>, rot=<-0.0044 0.0485 0.0924 -0.9945>}
 index_02_r {parent=index_01_r, loc=<-3.3587 0.3004 -0.0452>, rot=<-0.002 0.0027 0.3225 -0.9466>}
 index_03_r {parent=index_02_r, loc=<-2.5578 0.123 0.0252>, rot=<-0.002 0.0067 0.2558 -0.9667>}
 middle_metacarpal_r {parent=hand_r, loc=<-2.1856 0.1725 0.5984>, rot=<0.006 0.0168 -0.0329 -0.9993>}
 middle_01_r {parent=middle_metacarpal_r, loc=<-7.3275 0.2447 -0.0129>, rot=<-0.0208 0.0042 0.0858 -0.9961>}
 middle_02_r {parent=middle_01_r, loc=<-4.0123 -0.1401 -0.1175>, rot=<-0.0218 0.0258 0.3512 -0.9357>}
 middle_03_r {parent=middle_02_r, loc=<-2.6077 -0.0534 -0.1373>, rot=<0.0082 0.0188 0.1708 -0.9851>}
 ring_metacarpal_r {parent=hand_r, loc=<-2.2074 0.3625 -1.3638>, rot=<-0.0284 0.0722 -0.0458 -0.9959>}
 ring_01_r {parent=ring_metacarpal_r, loc=<-7.0295 0.6197 0.1446>, rot=<-0.0489 -0.0252 0.1892 -0.9804>}
 ring_02_r {parent=ring_01_r, loc=<-3.6174 -0.4647 0.1027>, rot=<0 0 0.2806 -0.9598>}
 ring_03_r {parent=ring_02_r, loc=<-2.4152 -0.114 -0.0199>, rot=<0 0 0.1189 -0.9929>}
 pinky_metacarpal_r {parent=hand_r, loc=<-1.7616 0.9753 -3.3814>, rot=<-0.0397 0.0814 -0.069 -0.9935>}
 pinky_01_r {parent=pinky_metacarpal_r, loc=<-7.1709 0.4831 0.1094>, rot=<-0.0654 -0.0035 0.2242 -0.9723>}
 pinky_02_r {parent=pinky_01_r, loc=<-2.8274 -0.0893 0.1043>, rot=<0.0037 0.0011 0.2854 -0.9584>}
 pinky_03_r {parent=pinky_02_r, loc=<-2.2046 -0.0307 0.0078>, rot=<-0.0002 0.0098 0.1445 -0.9895>}
 weapon_r {parent=hand_r, loc=<0.1974 3.4111 -0.3811>, rot=<0 0 0 1>}
 deform_wrist_dn_base_r {parent=hand_r, loc=<0 3.4172 0>, rot=<0 0 0 -1>}
 deform_wrist_dn_r {parent=deform_wrist_dn_base_r, loc=<0 0 0>, rot=<0 0 0 -1>}
 deform_wrist_up_base_r {parent=hand_r, loc=<-0.0001 -2.7804 0>, rot=<0 0 0 -1>}
 deform_wrist_up_r {parent=deform_wrist_up_base_r, loc=<0 0 0>, rot=<0 0 0 -1>}
 lowerarm_twist_02_r {parent=lowerarm_r, loc=<-9.6199 0 0>, rot=<0 0 0 -1>}
 lowerarm_twist_01_r {parent=lowerarm_r, loc=<-17.8549 0 0>, rot=<0 0 0 -1>}
 deform_elbow_in_base_r {parent=lowerarm_r, loc=<-0.2114 -5.6449 -0.827>, rot=<0 0 0 -1>}
 deform_elbow_in_r {parent=deform_elbow_in_base_r, loc=<0 0 0>, rot=<0 0 0 1>}
 deform_elbow_out_base_r {parent=lowerarm_r, loc=<-0.0001 7.5129 0.0002>, rot=<0 0 0 -1>}
 deform_elbow_out_r {parent=deform_elbow_out_base_r, loc=<0 0 0>, rot=<0 0 0 1>}
 upperarm_twist_01_r {parent=upperarm_r, loc=<-7.9796 -0.0843 0.6195>, rot=<0 0 0 -1>}
 deform_upperarm_bk_base_r {parent=upperarm_twist_01_r, loc=<6.8985 13.3826 1.3257>, rot=<0 0 0 1>}
 deform_upperarm_bk_r {parent=deform_upperarm_bk_base_r, loc=<0 0 0>, rot=<0 0 0 1>}
 deform_upperarm_fr_base_r {parent=upperarm_twist_01_r, loc=<8.3485 -10.7117 2.1456>, rot=<0 0 0 1>}
 deform_upperarm_fr_r {parent=deform_upperarm_fr_base_r, loc=<0 0 0>, rot=<0 0 0 1>}
 deform_upperarm_up_base_r {parent=upperarm_twist_01_r, loc=<11.2015 1.2234 -8.7788>, rot=<0 0 0 1>}
 deform_upperarm_up_r {parent=deform_upperarm_up_base_r, loc=<0 0 0>, rot=<0 0 0 1>}
 upperarm_twist_02_r {parent=upperarm_r, loc=<-19.5059 -0.0843 0.6195>, rot=<0 0 0 -1>}
 deform_bicep_base_r {parent=upperarm_twist_02_r, loc=<0.0002 -8.0494 -0.0003>, rot=<0 0 0 1>}
 deform_bicep_r {parent=deform_bicep_base_r, loc=<0 0 0>, rot=<0 0 0 1>}
 attach_cape {parent=spine_05, loc=<9.017 -11.714 0>, rot=<0.0616 -0.7044 -0.0616 -0.7044>}
 cape {parent=attach_cape, type=Null, loc=<0 -5.4545 -8.241>, rot=<-0.0575 0 0 0.9983>}
 attach_backpack {parent=spine_05, loc=<-0.123 -13.473 0>, rot=<-0.0093 -0.707 0.0093 -0.707>}
 Backpack_BR {parent=attach_backpack, type=Null, loc=<0 -7.3222 0.8832>, rot=<-0.0797 0 0 0.9968>}
 dyn_neck_r_base {parent=spine_05, loc=<13.8305 0.3706 19.689>, rot=<0.8194 0.0329 0.5718 -0.0229>}
 dyn_neck_r {parent=dyn_neck_r_base, loc=<0 0 0>, rot=<0 0 0 1>}
 dyn_neck_bk_base {parent=spine_05, loc=<15.6576 -14.1045 0>, rot=<0.4795 0.5196 0.5196 0.4795>}
 dyn_neck_bk {parent=dyn_neck_bk_base, loc=<0 0 0>, rot=<0 0 0 1>}
 dyn_neck_l_base {parent=spine_05, loc=<13.8305 0.3706 -19.6892>, rot=<0.0229 -0.5718 -0.0329 -0.8194>}
 dyn_neck_l {parent=dyn_neck_l_base, loc=<0 0 0>, rot=<0 0 0 1>}
 dyn_neck_fr_base {parent=spine_05, loc=<5.161 14.8377 0>, rot=<0.4795 0.5196 0.5196 0.4795>}
 dyn_neck_fr {parent=dyn_neck_fr_base, loc=<0 0 0>, rot=<0 0 0 1>}
 dyn_scarf_fr_base {parent=spine_05, loc=<11.4686 15.2678 12.4191>, rot=<-0.0056 -0.9445 -0.2748 0.1797>}
 dyn_scarf_fr {parent=dyn_scarf_fr_base, loc=<0 0 0>, rot=<0 0 0 1>}
 dyn_scarf_fr_01 {parent=dyn_scarf_fr_base, loc=<0 0 0>, rot=<0 0 0 1>}
 dyn_scarf_fr_02 {parent=dyn_scarf_fr_01, loc=<8.5084 0 0>, rot=<0 0 0 1>}
 dyn_scarf_bk_base {parent=spine_05, loc=<9.9051 -14.6104 -12.9372>, rot=<0.3751 -0.9053 -0.1271 -0.1537>}
 dyn_scarf_bk {parent=dyn_scarf_bk_base, loc=<0 0 0>, rot=<0 0 0 1>}
 deform_pec_aim_base_l {parent=spine_04, loc=<5.0227 13.6719 -2.6508>, rot=<0.0672 -0.5201 0.3228 -0.7879>}
 deform_pec_aim_l {parent=deform_pec_aim_base_l, loc=<0 0 0>, rot=<0 0 0 1>}
 deform_pec_base_l {parent=deform_pec_aim_l, loc=<16.3279 0 0>, rot=<0 0 0 1>}
 deform_pec_l {parent=deform_pec_base_l, loc=<0 0 0>, rot=<0 0 0 1>}
 deform_pec_aim_base_r {parent=spine_04, loc=<5.0225 13.6718 2.6508>, rot=<-0.5201 -0.0672 0.7879 0.3228>}
 deform_pec_aim_r {parent=deform_pec_aim_base_r, loc=<0 0 0>, rot=<0 0 0 1>}
 deform_pec_base_r {parent=deform_pec_aim_r, loc=<-16.3279 0 0>, rot=<0 0 0 -1>}
 deform_pec_r {parent=deform_pec_base_r, loc=<0 0 0.0002>, rot=<0 0 0 -1>}
 dyn_armpit_r_base {parent=spine_04, loc=<0.018 -6.7913 23.3317>, rot=<0.7065 0.0294 0.7065 -0.0294>}
 dyn_armpit_r {parent=dyn_armpit_r_base, loc=<0 0 0>, rot=<0 0 0 1>}
 dyn_armpit_l_base {parent=spine_04, loc=<0.018 -6.7913 -23.3319>, rot=<0.0294 -0.7065 -0.0294 -0.7065>}
 dyn_armpit_l {parent=dyn_armpit_l_base, loc=<0 0 0>, rot=<0 0 0 1>}
 deform_lat_l_base {parent=spine_03, loc=<2.5561 -10.3716 -17.6918>, rot=<0 0 0 -1>}
 deform_lat_l {parent=deform_lat_l_base, loc=<0 0 0>, rot=<0 0 0 1>}
 deform_lat_r_base {parent=spine_03, loc=<2.5562 -10.3716 17.6918>, rot=<0 0 1 0>}
 deform_lat_r {parent=deform_lat_r_base, loc=<0 0 0>, rot=<0 0 0 -1>}
 dyn_belly_l_base {parent=spine_01, loc=<-0.0163 -1.2296 -20.6786>, rot=<0.0114 -0.707 -0.0114 -0.707>}
 dyn_belly_l {parent=dyn_belly_l_base, loc=<0 0 0>, rot=<0 0 0 1>}
 dyn_belly_r_base {parent=spine_01, loc=<-0.0163 -1.2296 20.679>, rot=<0.707 0.0114 0.707 -0.0114>}
 dyn_belly_r {parent=dyn_belly_r_base, loc=<0 0 0>, rot=<0 0 0 1>}
 dyn_belly_m_base {parent=spine_01, loc=<-1.6584 14.3737 -0.0578>, rot=<0.4919 0.508 0.508 0.4919>}
 dyn_belly_m {parent=dyn_belly_m_base, loc=<0 0 0>, rot=<0 0 0 1>}
 deform_glute_base_l {parent=pelvis, loc=<-16.99 -11.0586 -11.4158>, rot=<0.0275 0.0501 -0.4472 -0.8926>}
 deform_glute_l {parent=deform_glute_base_l, loc=<0 0 0>, rot=<0 0 0 1>}
 deform_glute_base_r {parent=pelvis, loc=<-16.9958 -11.059 11.416>, rot=<0.0501 -0.0275 0.8926 -0.4472>}
 deform_glute_r {parent=deform_glute_base_r, loc=<0 0 0>, rot=<0 0 0 1>}
 thigh_r {parent=pelvis, loc=<-1.8714 -0.0109 11.9936>, rot=<0.0572 -0.0012 0.9982 0.0155>}
 calf_r {parent=thigh_r, loc=<41.918 0.2246 -0.0926>, rot=<0.0006 -0.0161 -0.0391 -0.9991>}
 calf_twist_01_r {parent=calf_r, loc=<32.9336 0.0967 -0.0397>, rot=<0.0237 0.0095 -0.0035 -0.9997>}
 calf_twist_02_r {parent=calf_r, loc=<15.9268 0.0967 -0.0397>, rot=<0.0237 0.0095 -0.0035 -0.9997>}
 foot_r {parent=calf_r, loc=<38.7058 -1.2438 -0.2453>, rot=<0.0975 -0.0474 0.0317 -0.9936>}
 ball_r {parent=foot_r, loc=<6.1614 -13.1469 -0.6807>, rot=<0.0001 -0.0001 -0.7063 -0.7079>}
 dyn_flip_flopper_r {parent=ball_r, loc=<5.6004 -1.5117 0>, rot=<0 -0.7071 0.7071 0>}
 dyn_foot_bk_r_base {parent=foot_r, loc=<-1.5234 4.2103 -0.1706>, rot=<0.5046 0.499 -0.4955 -0.5009>}
 dyn_foot_bk_r {parent=dyn_foot_bk_r_base, loc=<0 0 0>, rot=<0 0 0 1>}
 deform_knee_in_base_r {parent=calf_r, loc=<-1.2981 8.0356 0.0496>, rot=<0 0 0 1>}
 deform_knee_in_r {parent=deform_knee_in_base_r, loc=<0 0 0>, rot=<0 0 0 1>}
 deform_knee_out_base_r {parent=calf_r, loc=<0 -6.8739 -0.0001>, rot=<0 0 0 1>}
 deform_knee_out_r {parent=deform_knee_out_base_r, loc=<0 0 0>, rot=<0 0 0 1>}
 thigh_twist_01_r {parent=thigh_r, loc=<21.6303 0 0>, rot=<0.0013 0 0 -1>}
 deform_groin_base_r {parent=thigh_r, loc=<2.191 -8.5661 1.7442>, rot=<0 0 0 -1>}
 deform_groin_r {parent=deform_groin_base_r, loc=<0 0 0>, rot=<0 0 0 -1>}
 thigh_l {parent=pelvis, loc=<-1.8716 -0.0109 -11.9936>, rot=<0.0012 0.0572 0.0155 -0.9982>}
 calf_l {parent=thigh_l, loc=<-41.9178 -0.2246 0.0926>, rot=<0.0006 -0.0161 -0.0391 -0.9991>}
 calf_twist_01_l {parent=calf_l, loc=<-32.9335 -0.0967 0.0397>, rot=<0.0237 0.0095 -0.0035 -0.9997>}
 calf_twist_02_l {parent=calf_l, loc=<-15.9268 -0.0967 0.0397>, rot=<0.0237 0.0095 -0.0035 -0.9997>}
 foot_l {parent=calf_l, loc=<-38.7057 1.2438 0.2452>, rot=<0.0975 -0.0474 0.0317 -0.9936>}
 ball_l {parent=foot_l, loc=<-6.1614 13.1469 0.6807>, rot=<0.0001 -0.0001 -0.7063 -0.7079>}
 dyn_flip_flopper_r1 {parent=ball_l, loc=<-5.6004 1.5118 0>, rot=<0 -0.7071 0.7071 0>}
 dyn_flip_flopper_l {parent=ball_l, loc=<-5.6004 1.5118 0>, rot=<0 -0.7071 0.7071 0>}
 dyn_foot_bk_l_base {parent=foot_l, loc=<1.5232 -4.2104 0.1709>, rot=<0.4685 -0.5295 0.4738 -0.5251>}
 dyn_foot_bk_l {parent=dyn_foot_bk_l_base, loc=<0 0 0>, rot=<0 0 0 1>}
 deform_knee_in_base_l {parent=calf_l, loc=<1.298 -8.0356 -0.0496>, rot=<0 0 0 1>}
 deform_knee_in_l {parent=deform_knee_in_base_l, loc=<0 0 0>, rot=<0 0 0 1>}
 deform_knee_out_base_l {parent=calf_l, loc=<0 6.8739 0>, rot=<0 0 0 1>}
 deform_knee_out_l {parent=deform_knee_out_base_l, loc=<0 0 0>, rot=<0 0 0 1>}
 thigh_twist_01_l {parent=thigh_l, loc=<-21.6303 0 0>, rot=<0.0013 0 0 -1>}
 dyn_legPouch_base {parent=thigh_twist_01_l, loc=<-2.8919 6.7113 -6.616>, rot=<-0.346 -0.6353 -0.3646 -0.5863>}
 dyn_legPouch_01 {parent=dyn_legPouch_base, loc=<0 0 0>, rot=<0 0 0 1>}
 deform_groin_base_l {parent=thigh_l, loc=<-2.1908 8.5661 -1.7442>, rot=<0 0 0 -1>}
 deform_groin_l {parent=deform_groin_base_l, loc=<0 0 0>, rot=<0 0 0 1>}
 dyn_waistbelt_m_base {parent=pelvis, loc=<-6.5813 15.0256 -0.0578>, rot=<0.4249 0.5652 0.5652 0.4249>}
 dyn_waistbelt_m {parent=dyn_waistbelt_m_base, loc=<0 0 0>, rot=<0 0 0 1>}
 dyn_waistbelt_ml_base {parent=pelvis, loc=<-6.341 9.8801 -16.7434>, rot=<-0.2251 -0.6963 -0.3686 -0.5733>}
 dyn_waistbelt_ml {parent=dyn_waistbelt_ml_base, loc=<0 0 0>, rot=<0 0 0 1>}
 dyn_waistbelt_l_base {parent=pelvis, loc=<-4.5765 -0.5673 -21.9857>, rot=<0.0785 -0.7353 -0.0283 -0.6726>}
 dyn_waistbelt_l {parent=dyn_waistbelt_l_base, loc=<0 0 0>, rot=<0 0 0 1>}
 dyn_waistbelt_mr_base {parent=pelvis, loc=<-6.3411 9.8801 16.7434>, rot=<0.2251 0.6963 -0.3686 -0.5733>}
 dyn_waistbelt_mr {parent=dyn_waistbelt_mr_base, loc=<0 0 0>, rot=<0 0 0 1>}
 dyn_waistbelt_r_base {parent=pelvis, loc=<-4.5765 -0.5673 21.9857>, rot=<-0.0785 0.7353 -0.0283 -0.6726>}
 dyn_waistbelt_r {parent=dyn_waistbelt_r_base, loc=<0 0 0>, rot=<0 0 0 -1>}
 attach {parent=root, loc=<0 0 0>, rot=<0 0 0 1>}
 ik_hand_root {parent=root, loc=<0 0 0>, rot=<0 0 0 1>}
 ik_hand_gun {parent=ik_hand_root, loc=<-64.462 4.8638 101.1961>, rot=<0.6084 0.1793 0.4613 -0.6205>}
 ik_hand_r {parent=ik_hand_gun, loc=<0 0 0>, rot=<0 0 0 1>}
 ik_hand_l {parent=ik_hand_gun, loc=<65.7708 101.9258 43.6659>, rot=<-0.5102 -0.7906 -0.3387 0>}
 ik_foot_root {parent=root, loc=<0 0 0>, rot=<0 0 0 1>}
 ik_foot_l {parent=ik_foot_root, loc=<19.6189 -3.2051 10.3339>, rot=<0.0753 0.7018 0.0677 -0.7051>}
 ik_foot_r {parent=ik_foot_root, loc=<-19.6189 -3.2051 10.3339>, rot=<0.7051 0.0677 -0.7018 0.0753>}
 root_fp {parent=root, loc=<0 0 0>, rot=<0 0 0 1>}
 camera_root {parent=root_fp, loc=<0 0 0>, rot=<0 0 0 1>}
 camera {parent=camera_root, loc=<0 0 0>, rot=<0 0 0 1>}
 attach_fp {parent=root_fp, loc=<0 0 0>, rot=<0 0 0 1>}
 attach_emote_01 {parent=root, loc=<0 0 0>, rot=<0 0 0 1>}
 attach_emote_02 {parent=root, loc=<0 0 0>, rot=<0 0 0 1>}
 attach_emote_03 {parent=root, loc=<0 0 0>, rot=<0 0 0 1>}
 warp_point_01 {parent=root, loc=<0 0 0>, rot=<0 0 0 1>}
 warp_point_02 {parent=root, loc=<0 0 0>, rot=<0 0 0 1>}
 warp_point_03 {parent=root, loc=<0 0 0>, rot=<0 0 0 1>}
 dyn_slope_foot_l {parent=root, loc=<0 0 0>, rot=<0 0 1 0>}
 dyn_slope_foot_r {parent=root, loc=<0 0 0>, rot=<0 0 1 0>}
 l_arm_ctrl {parent=root, loc=<0 0 -2.1991>, rot=<0 0 0.0545 0.9985>}
 r_arm_ctrl {parent=root, loc=<0 0 -2.1991>, rot=<0 0 0.0545 0.9985>}
 LeftArmScarfHide {type=Curve, loc=<0 0 0>, rot=<0 0 0 1>}
]
variables [
 NewVar {type=FRotator, default=(Pitch=0,Yaw=0,Roll=0)}
 SphericalPoseReader_OutputParam {type=float, default=0}
 isGauntletsEquip {type=bool, default=False}
 isPlayingEmote {type=bool, default=False}
]
