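# Physics asset for the left-leg dynamics of the large male ChubbyJingle player body, apparently driving a floppy foot piece.
# It has box bodies for the foot, a thin flip-flopper and a wide slope-foot slab whose top sits at its bone origin.
# One constraint joins the flip-flopper to the foot with swing1 and twist locked, leaving swing2 as a hinge-like axis.
# The slope-foot slab has no constraint.
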PhysicsAsset /BRCosmetics/Characters/Player/Male/Large/Bodies/M_LRG_ChubbyJingle/Meshes/M_LRG_ChubbyJingle_Dyn_leg_l
bodies {
 dyn_flip_flopper_l [{shapeType=Box, extentX=16.6875, extentY=9.6875, extentZ=2, center=<8.5683 0.0001 0.2546>}]
 dyn_slope_foot_l [{shapeType=Box, extentX=100, extentY=100, extentZ=8, center=<0 0 -4>}]
 foot_l [{shapeType=Box, extentX=42.3125, extentY=31, extentZ=18.5, center=<2.8771 1.6758 -0.0005>, rotation=<rot 0.0003 89.9996 89.9999>}]
}
constraints [{
  bone1Name=dyn_flip_flopper_l
  bone2Name=foot_l
  swing1LimitDegrees=45
  swing1Motion=Locked
  swing2LimitDegrees=30
  twistLimitDegrees=45
  twistMotion=Locked
 }]
