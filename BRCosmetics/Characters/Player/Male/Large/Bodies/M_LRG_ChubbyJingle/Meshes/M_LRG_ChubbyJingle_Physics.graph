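# Physics asset for the M_LRG_ChubbyJingle large male player body, apparently built for secondary motion on its scarf.
# Capsules on the upper spine, both clavicles and the right pec approximate the torso around the scarf bones.
# The front scarf is a two-box chain constrained to a tiny base capsule with tight twist limits.
# The back scarf is one capsule on its own base with locked twist, while clavicle_r links to spine_04 with free angular motion.
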PhysicsAsset /BRCosmetics/Characters/Player/Male/Large/Bodies/M_LRG_ChubbyJingle/Meshes/M_LRG_ChubbyJingle_Physics
bodies {
 clavicle_l [{
   shapeType=Capsule
   length=13.6053
   radius=13.1217
   shapeName=clavicle_l_capsule
   center=<11.6234 -1.7978 -4.2101>
   rotation=<rot 8.1656 -71.9818 -85.2479>
  }]
 clavicle_r [{
   shapeType=Capsule
   length=15
   radius=13.942
   shapeName=clavicle_r_capsule
   center=<-9.2858 1.2081 11.5202>
   rotation=<rot 0.6943 58.1707 -69.3463>
  }]
 deform_pec_r [{
   shapeType=Capsule
   length=15.3067
   radius=13.3373
   shapeName=deform_pec_r_capsule
   center=<6.9714 6.55 4.7413>
   rotation=<rot 12.3366 -86.2122 90.2095>
  }]
 dyn_scarf_bk [{
   shapeType=Capsule
   length=6.3808
   radius=5.3174
   shapeName=dyn_scarf_bk_capsule
   center=<7.0876 -2.344 1.1033>
   rotation=<rot 7.1937 -89.9816 0.1349>
  }]
 dyn_scarf_bk_base [{shapeType=Capsule, length=0.505, radius=0.505, shapeName=dyn_scarf_bk_base_capsule, rotation=<rot 0 -90 0>}]
 dyn_scarf_fr_01 [{
   shapeType=Box
   extentX=2.8526
   extentY=9.9067
   extentZ=11.7439
   shapeName=dyn_scarf_fr_01_box
   center=<4.1768 0.5464 0.298>
   rotation=<rot -4.5825 -90 4.0294>
  }]
 dyn_scarf_fr_02 [{
   shapeType=Box
   extentX=2.2915
   extentY=13.607
   extentZ=12.9031
   shapeName=dyn_scarf_fr_02_box
   center=<4.9333 1.1372 0.0521>
   rotation=<rot -1.2286 -78.4676 3.7418>
  }]
 dyn_scarf_fr_base [{shapeType=Capsule, length=0.505, radius=0.505, shapeName=dyn_scarf_fr_base_capsule, rotation=<rot 0 -90 0>}]
 spine_04 [{
   shapeType=Capsule
   length=25.6182
   radius=18.9486
   shapeName=spine_04_capsule
   center=<-2.6968 4.8437 -1.5118>
   rotation=<rot -26.171 -90.486 1.1631>
  }]
 spine_05 [{shapeType=Capsule, length=18.9875, radius=17.122, shapeName=spine_05_capsule, center=<-14.5684 1.1959 -1.0812>, rotation=<rot 0 180 0>}]
}
constraints [
 {bone1Name=dyn_scarf_bk, bone2Name=dyn_scarf_bk_base, swing1LimitDegrees=30, swing2LimitDegrees=5, twistLimitDegrees=45, twistMotion=Locked}
 {
  bone1Name=clavicle_r
  bone2Name=spine_04
  swing1LimitDegrees=45
  swing1Motion=Free
  swing2LimitDegrees=45
  swing2Motion=Free
  twistLimitDegrees=45
  twistMotion=Free
 }
 {bone1Name=dyn_scarf_fr_02, bone2Name=dyn_scarf_fr_01, swing1LimitDegrees=25, swing2LimitDegrees=15, twistLimitDegrees=3}
 {bone1Name=dyn_scarf_fr_01, bone2Name=dyn_scarf_fr_base, swing1LimitDegrees=25, swing2LimitDegrees=15, twistLimitDegrees=3}
]
