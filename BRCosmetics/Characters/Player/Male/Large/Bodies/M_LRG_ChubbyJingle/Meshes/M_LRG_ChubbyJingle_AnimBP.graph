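# Animation Blueprint for the M_LRG_ChubbyJingle cosmetic body, layering secondary motion and corrections onto the pose copied from its parent mesh.
# Slope-foot adjustment, front-end-only leg physics, head/neck rotation and a Control Rig feed scarf and pouch constraints.
# Backbling-dependent RigidBody setups follow, then limb bone corrections and Anim Dynamics belly jiggle.
# The EventGraph computes counter-scaled head/neck rotations, slope data, gauntlet/boot detection, foot-flop alphas and a scarf-hide curve.
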AnimBlueprint /BRCosmetics/Characters/Player/Male/Large/Bodies/M_LRG_ChubbyJingle/Meshes/M_LRG_ChubbyJingle_AnimBP
target_skeleton=&/M_LRG_ChubbyJingle_Skeleton
graphs [
 AnimGraph {nodes [
   Anim Dynamics - Bone: dyn_scarf_fr(Alpha=1, ExternalForce="0,0,0") {
    angularDampingOverride=0.999
    angularSpringConstant=1500
    bAngularSpring=true
    bOverrideAngularDamping=true
    bUseGravityOverride=true
    boundBone {boneName=dyn_scarf_fr}
    componentAppliedLinearAccClamp=<5000 5000 5000>
    componentLinearAccScale=<0.5 0.5 0.5>
    componentLinearVelScale=<0.5 0.5 0.5>
    gravityOverride {z=-980}
    physicsBodyDefinitions [{
      boundBone {boneName=dyn_scarf_fr}
      boxExtents=<10 10 10>
      constraintSetup {angularLimitsMax=<0 0 15>, angularLimitsMin=<0 0 0>, angularTarget=<0 0 0>}
      localJointOffset=<0.1 0 0>
     }]
   }
   Anim Dynamics - Bone: dyn_scarf_bk(Alpha=1, ExternalForce="0,0,0") {
    angularDampingOverride=0.999
    angularSpringConstant=1500
    bAngularSpring=true
    bOverrideAngularDamping=true
    bUseGravityOverride=true
    boundBone {boneName=dyn_scarf_bk}
    componentAppliedLinearAccClamp=<5000 5000 5000>
    componentLinearAccScale=<0.5 0.5 0.5>
    componentLinearVelScale=<0.5 0.5 0.5>
    gravityOverride {z=-980}
    physicsBodyDefinitions [{
      boundBone {boneName=dyn_scarf_bk}
      boxExtents=<10 10 10>
      constraintSetup {angularLimitsMax=<0 0 12>, angularLimitsMin=<0 0 -15>, angularTarget=<0 0 0>}
      localJointOffset=<0.1 0 0>
     }]
   }
   n1 = RigidBody!(ComponentPose=Local To Component(LocalPose=Use cached pose 'Pre phys'()), bAlphaBoolEnabled:bool, ExternalForce) -> (Pose) {
    alphaInputType=Bool
    bClampLinearTranslationLimitToRefPose=true
    bDefaultToSkeletalMeshPhysicsAsset=false
    bOverrideWorldGravity=true
    cachedBoundsScale=1.15
    overridePhysicsAsset=&/M_LRG_ChubbyJingle_Backpack_Physics
    simulationSpace=BaseBoneSpace
    alphaBoolBlend {blendOption=Linear}
    baseBoneRef {boneName=root}
    overrideWorldGravity {z=-980}
    simSpaceSettings {
     dampingAlpha=0
     maxAngularAcceleration=20
     maxAngularVelocity=10
     maxLinearAcceleration=200
     maxLinearVelocity=300
     worldAlpha=1
     externalLinearDragV=<0.1 0.1>
    }
   }
   n2 = "Transform (Modify) Bone - Bone: dyn_scarf_bk"!(ComponentPose=@n1, Alpha=1, Rotation="-9.032,3.071,-17.503", Scale="1.000,0.610,1.000") -> (Pose) {rotationMode=BMM_Additive, scaleMode=BMM_Additive, boneToModify {boneName=dyn_scarf_bk}, rotation=<rot -9.032 3.071 -17.503>, scale {y=0.61}}
   n3 = Component To Local(ComponentPose=@n2)
   n4 = RigidBody(ComponentPose=Local To Component(LocalPose=Use cached pose 'Pre phys'())) {
    alphaInputType=Bool
    bClampLinearTranslationLimitToRefPose=true
    bDefaultToSkeletalMeshPhysicsAsset=false
    bOverrideWorldGravity=true
    cachedBoundsScale=1.15
    overridePhysicsAsset=&/M_LRG_ChubbyJingle_Physics
    simulationSpace=BaseBoneSpace
    alphaBoolBlend {blendOption=Linear}
    baseBoneRef {boneName=root}
    overrideWorldGravity {z=-980}
    simSpaceSettings {
     dampingAlpha=0
     maxAngularAcceleration=20
     maxAngularVelocity=10
     maxLinearAcceleration=200
     maxLinearVelocity=300
     worldAlpha=1
     externalLinearDragV=<0.1 0.1>
    }
   }
   n5 = Component To Local(ComponentPose=@n4)
   n6 = Get isWearingBackbling()
   n7 = "Transform (Modify) Bone - Bone: dyn_scarf_fr_01"!(
    ComponentPose=Local To Component(LocalPose=Blend Poses by bool(BlendPose_0=@n3, BlendPose_1=@n5, bActiveValue=@n6))
    bAlphaBoolEnabled:bool=NOT Boolean(A=Get bRigidBodyEnabled())
    Translation="0,0,0"
    Rotation="-7.436,-5.985,-7.326"
    Scale="1,1,1"
   ) -> (Pose) {alphaInputType=Bool, rotationMode=BMM_Additive, boneToModify {boneName=dyn_scarf_fr_01}, rotation=<rot -7.436 -5.985 -7.326>}
   n8 = "Transform (Modify) Bone - Bone: calf_l"!(ComponentPose=@n7, bAlphaBoolEnabled:bool=True, Translation="0,0,0", Rotation="0,0,0", Scale="1,0.5,0.5") -> (Pose) {alphaInputType=Bool, scaleMode=BMM_Additive, alphaBoolBlend {blendOption=Linear}, boneToModify {boneName=calf_l}, scale {y=0.5, z=0.5}}
   n9 = "Transform (Modify) Bone - Bone: calf_r"!(ComponentPose=@n8, bAlphaBoolEnabled:bool=True, Translation="0,0,0", Rotation="0,0,0", Scale="1,0.5,0.5") -> (Pose) {alphaInputType=Bool, scaleMode=BMM_Additive, alphaBoolBlend {blendOption=Linear}, boneToModify {boneName=calf_r}, scale {y=0.5, z=0.5}}
   n10 = "Transform (Modify) Bone - Bone: upperarm_twist_02_l"!(ComponentPose=@n9, bAlphaBoolEnabled:bool=True, Translation="0,0,0", Rotation="0,0,0", Scale="1,0.9,0.9") -> (Pose) {
    alphaInputType=Bool
    scaleMode=BMM_Additive
    scaleSpace=BCS_BoneSpace
    translationMode=BMM_Additive
    translationSpace=BCS_BoneSpace
    boneToModify {boneName=upperarm_twist_02_l}
    scale {y=0.9, z=0.9}
   }
   n11 = "Transform (Modify) Bone - Bone: lowerarm_l"!(ComponentPose=@n10, bAlphaBoolEnabled:bool=True, Translation="0,0,0", Rotation="0,0,0", Scale="1,0.8,0.8") -> (Pose) {
    alphaInputType=Bool
    scaleMode=BMM_Additive
    scaleSpace=BCS_BoneSpace
    translationMode=BMM_Additive
    translationSpace=BCS_BoneSpace
    boneToModify {boneName=lowerarm_l}
    scale {y=0.8, z=0.8}
   }
   n12 = "Transform (Modify) Bone - Bone: lowerarm_twist_01_l"!(ComponentPose=@n11, bAlphaBoolEnabled:bool=True, Translation="0,5,8", Rotation="0,0,0", Scale="1,0.7,0.7") -> (Pose) {
    alphaInputType=Bool
    scaleMode=BMM_Additive
    scaleSpace=BCS_BoneSpace
    translationMode=BMM_Additive
    translationSpace=BCS_BoneSpace
    boneToModify {boneName=lowerarm_twist_01_l}
    scale {y=0.7, z=0.7}
    translation {y=5, z=8}
   }
   n13 = "Transform (Modify) Bone - Bone: lowerarm_twist_02_l"!(ComponentPose=@n12, bAlphaBoolEnabled:bool=True, Translation="0,-2,0", Rotation="0,0,0", Scale="1,0.6,0.7") -> (Pose) {
    alphaInputType=Bool
    scaleMode=BMM_Additive
    scaleSpace=BCS_BoneSpace
    translationMode=BMM_Additive
    translationSpace=BCS_BoneSpace
    boneToModify {boneName=lowerarm_twist_02_l}
    scale {y=0.6, z=0.7}
    translation {y=-2}
   }
   n14 = "Transform (Modify) Bone - Bone: upperarm_twist_02_r"!(ComponentPose=@n13, bAlphaBoolEnabled:bool=True, Translation="0,0,0", Rotation="0,0,0", Scale="1,0.9,0.9") -> (Pose) {
    alphaInputType=Bool
    scaleMode=BMM_Additive
    scaleSpace=BCS_BoneSpace
    translationMode=BMM_Additive
    translationSpace=BCS_BoneSpace
    boneToModify {boneName=upperarm_twist_02_r}
    scale {y=0.9, z=0.9}
   }
   n15 = "Transform (Modify) Bone - Bone: lowerarm_r"!(ComponentPose=@n14, bAlphaBoolEnabled:bool=True, Translation="0,3.5,0", Rotation="0.034,8.250,0.026", Scale="1,0.75,0.8") -> (Pose) {
    alphaInputType=Bool
    rotationMode=BMM_Additive
    rotationSpace=BCS_BoneSpace
    scaleMode=BMM_Additive
    scaleSpace=BCS_BoneSpace
    translationMode=BMM_Additive
    translationSpace=BCS_BoneSpace
    boneToModify {boneName=lowerarm_r}
    rotation=<rot 0.034 8.25 0.026>
    scale {y=0.75, z=0.8}
    translation {y=3.5}
   }
   n16 = "Transform (Modify) Bone - Bone: lowerarm_twist_01_r"!(ComponentPose=@n15, bAlphaBoolEnabled:bool=True, Translation="-17.673,8.270,5.027", Rotation="-27.337,-19.270,-47.360", Scale="1,0.7,0.7") -> (Pose) {
    alphaInputType=Bool
    rotationMode=BMM_Additive
    rotationSpace=BCS_BoneSpace
    scaleMode=BMM_Additive
    scaleSpace=BCS_BoneSpace
    translationMode=BMM_Additive
    translationSpace=BCS_BoneSpace
    boneToModify {boneName=lowerarm_twist_01_r}
    rotation=<rot -27.337 -19.27 -47.36>
    scale {y=0.7, z=0.7}
    translation=<-17.673 8.27 5.027>
   }
   n17 = "Transform (Modify) Bone - Bone: lowerarm_twist_02_r"!(ComponentPose=@n16, bAlphaBoolEnabled:bool=True, Translation="0.084,1.5,-2.5", Rotation="0,0,0", Scale="1,0.65,0.7") -> (Pose) {
    alphaInputType=Bool
    scaleMode=BMM_Additive
    scaleSpace=BCS_BoneSpace
    translationMode=BMM_Additive
    translationSpace=BCS_BoneSpace
    boneToModify {boneName=lowerarm_twist_02_r}
    scale {y=0.65, z=0.7}
    translation=<0.084 1.5 -2.5>
   }
   n18 = "Constraint - dyn_belly_m_base (PPPP)"!(
    ComponentPose="Transform (Modify) Bone - Bone: dyn_scarf_fr_02"(ComponentPose=@n17, Alpha=1, Translation="0,0,0", Rotation="0,0,0", Scale="0.5,0.5,0.5") {scaleMode=BMM_Additive, boneToModify {boneName=dyn_scarf_fr_02}, scale=<0.5 0.5 0.5>}
    Alpha=1
    ConstraintWeights_0=0
    ConstraintWeights_1:float=0
    ConstraintWeights_2:float=0
    ConstraintWeights_3:float=1
   ) -> (Pose) {
    boneToModify {boneName=dyn_belly_m_base}
    constraintSetup [{targetBone {boneName=spine_01}}, {targetBone {boneName=spine_02}}, {targetBone {boneName=spine_03}}, {targetBone {boneName=pelvis}}]
    constraintWeights [1,1,1,1]
   }
   n19 = Anim Dynamics - Bone: dyn_belly_m_base(ComponentPose=@n18, Alpha=0.9, ExternalForce="0,0,0") {
    angularDampingOverride=0.9
    angularSpringConstant=5000
    bLinearSpring=true
    bOverrideLinearDamping=true
    bUseGravityOverride=true
    gravityScale=0.35
    linearDampingOverride=0.95
    linearSpringConstant=150
    simulationSpace=RootRelative
    boundBone {boneName=dyn_belly_m_base}
    componentAppliedLinearAccClamp=<0 0 0>
    gravityOverride {z=-980}
    physicsBodyDefinitions [{
      boundBone {boneName=dyn_belly_m_base}
      boxExtents=<35 35 35>
      constraintSetup {angularLimitsMax=<0 0 0>, angularLimitsMin=<0 0 0>, angularTarget=<1 0 0>, linearAxesMax=<1 1 1>, linearAxesMin=<0 -1 -1>}
      localJointOffset=<3 0 0>
     }]
    relativeSpaceBone {boneName=pelvis}
    simSpaceSettings {maxAngularAcceleration=15, maxAngularVelocity=100, simSpaceAngularAlpha=1}
   }
   n20 = Anim Dynamics - Bone: dyn_belly_m_base(ComponentPose=@n19, Alpha=0.9, ExternalForce="0,0,0") {
    angularDampingOverride=0.9999
    angularSpringConstant=25000
    bAngularSpring=true
    bOverrideAngularDamping=true
    bUseGravityOverride=true
    gravityScale=0.35
    simulationSpace=RootRelative
    boundBone {boneName=dyn_belly_m_base}
    componentAppliedLinearAccClamp=<0 0 0>
    gravityOverride {z=-980}
    physicsBodyDefinitions [{
      boundBone {boneName=dyn_belly_m_base}
      boxExtents=<35 35 35>
      constraintSetup {angularLimitsMax=<15 15 15>, angularLimitsMin=<-15 -15 -15>, angularTarget=<1 0 0>}
      localJointOffset=<3 0 0>
     }]
    relativeSpaceBone {boneName=pelvis}
    simSpaceSettings {maxAngularAcceleration=15, maxAngularVelocity=100, simSpaceAngularAlpha=1}
   }
   Output Pose(Result=Component To Local(ComponentPose=@n20)) {layerGroup=None}
   n21 = NOT Boolean(A=Get bCanPlayCustomAnimations() -> (bCanPlayCustomAnimations:bool))
   n22 = "Transform (Modify) Bone - Bone: dyn_slope_foot_l"!(
    ComponentPose=Local To Component(LocalPose=Use cached pose 'Base'())
    bAlphaBoolEnabled:bool=@n21
    Translation="0,0,0"
    Rotation="0,0,0"
    Scale="1,1,1"
   ) -> (Pose) {
    alphaInputType=Bool
    rotationMode=BMM_Replace
    rotationSpace=BCS_WorldSpace
    translationMode=BMM_Replace
    translationSpace=BCS_WorldSpace
    alphaBoolBlend {blendOption=Linear}
    boneToModify {boneName=dyn_slope_foot_l}
   }
   n23 = "Transform (Modify) Bone - Bone: dyn_slope_foot_r"!(ComponentPose=@n22, bAlphaBoolEnabled:bool=@n21, Translation="0,0,0", Rotation="0,0,0", Scale="1,1,1") -> (Pose) {
    alphaInputType=Bool
    rotationMode=BMM_Replace
    rotationSpace=BCS_WorldSpace
    translationMode=BMM_Replace
    translationSpace=BCS_WorldSpace
    alphaBoolBlend {blendOption=Linear}
    boneToModify {boneName=dyn_slope_foot_r}
   }
   n24 = Component To Local(ComponentPose=@n23)
   n25 = NOT Boolean(A=Get bCanPlayCustomAnimations())
   n26 = "Transform (Modify) Bone - Bone: dyn_slope_foot_l"(
    ComponentPose=Local To Component(LocalPose=Use cached pose 'Base'())
    bAlphaBoolEnabled=@n25
    Translation="0,0,0"
    Rotation="0,0,0"
    Scale="1,1,1"
   ) {
    alphaInputType=Bool
    rotationMode=BMM_Replace
    rotationSpace=BCS_WorldSpace
    translationMode=BMM_Replace
    translationSpace=BCS_WorldSpace
    alphaBoolBlend {blendOption=Linear}
    boneToModify {boneName=dyn_slope_foot_l}
   }
   n27 = "Transform (Modify) Bone - Bone: dyn_slope_foot_r"(ComponentPose=@n26, bAlphaBoolEnabled=@n25, Translation="0,0,0", Rotation="0,0,0", Scale="1,1,1") {
    alphaInputType=Bool
    rotationMode=BMM_Replace
    rotationSpace=BCS_WorldSpace
    translationMode=BMM_Replace
    translationSpace=BCS_WorldSpace
    alphaBoolBlend {blendOption=Linear}
    boneToModify {boneName=dyn_slope_foot_r}
   }
   n28 = RigidBody!(ComponentPose=@n27, Alpha:float, ExternalForce) -> (Pose) {
    bClampLinearTranslationLimitToRefPose=true
    bOverrideWorldGravity=true
    cachedBoundsScale=1.15
    overridePhysicsAsset=&/M_LRG_ChubbyJingle_Dyn_leg_l
    simulationSpace=BaseBoneSpace
    alphaBoolBlend {blendOption=Linear}
    baseBoneRef {boneName=calf_l}
    componentAppliedLinearAccClamp=<3000 3000 3000>
    overrideWorldGravity {z=-980}
    simSpaceSettings {
     dampingAlpha=0
     maxAngularAcceleration=500
     maxAngularVelocity=3
     maxLinearAcceleration=400
     maxLinearVelocity=250
     worldAlpha=1
     externalLinearDragV=<0.1 0.1>
    }
   }
   n29 = RigidBody(ComponentPose=@n28) {
    bClampLinearTranslationLimitToRefPose=true
    bOverrideWorldGravity=true
    cachedBoundsScale=1.15
    overridePhysicsAsset=&/M_LRG_ChubbyJingle_Dyn_leg_r
    simulationSpace=BaseBoneSpace
    alphaBoolBlend {blendOption=Linear}
    baseBoneRef {boneName=calf_r}
    componentAppliedLinearAccClamp=<3000 3000 3000>
    overrideWorldGravity {z=-980}
    simSpaceSettings {
     dampingAlpha=0
     maxAngularAcceleration=500
     maxAngularVelocity=3
     maxLinearAcceleration=400
     maxLinearVelocity=250
     worldAlpha=1
     externalLinearDragV=<0.1 0.1>
    }
   }
   n30 = Component To Local(ComponentPose=@n29)
   n31 = Get bIsInFrontEnd()
   n32 = "Transform (Modify) Bone - Bone: head"!(
    ComponentPose=Local To Component(LocalPose=Blend Poses by bool(BlendPose_0=@n24, BlendPose_1=@n30, bActiveValue=@n31))
    Alpha=1
    Rotation="0,0,0"
   ) -> (Pose) {
    rotationMode=BMM_Additive
    rotationSpace=BCS_BoneSpace
    translationMode=BMM_Additive
    alphaBoolBlend {blendOption=Linear}
    boneToModify {boneName=head}
   }
   n33 = "Transform (Modify) Bone - Bone: neck_02"!(ComponentPose=@n32, Alpha=1, Rotation="0,0,0") -> (Pose) {
    rotationMode=BMM_Additive
    rotationSpace=BCS_BoneSpace
    translationMode=BMM_Additive
    alphaBoolBlend {blendOption=Linear}
    boneToModify {boneName=neck_02}
   }
   n34 = "Transform (Modify) Bone - Bone: neck_01"!(ComponentPose=@n33, Alpha=1, Rotation="0,0,0") -> (Pose) {
    rotationMode=BMM_Additive
    rotationSpace=BCS_BoneSpace
    translationMode=BMM_Additive
    alphaBoolBlend {blendOption=Linear}
    boneToModify {boneName=neck_01}
   }
   n35 = Control Rig!(
    Source=Component To Local(ComponentPose=@n34)
    Alpha
    isPlayingEmote:bool=Get bIsPlayingEmote()
    isGauntletsEquip:bool=Get GauntlesEquip() -> (GauntlesEquip:bool)
   ) -> (Pose) {controlRigAssetReference {blueprintRigClass=&/M_LRG_ChubbyJingle_CtrlRig.M_LRG_ChubbyJingle_CtrlRig_C}}
   n36 = "Constraint - dyn_legPouch_base (PPPPP)"!(
    ComponentPose=Local To Component(LocalPose=@n35)
    Alpha=1
    ConstraintWeights_0=0.0183
    ConstraintWeights_1:float=0.3005
    ConstraintWeights_2:float=0.0048
    ConstraintWeights_3:float=0.647
    ConstraintWeights_4:float=0.0295
   ) -> (Pose) {
    boneToModify {boneName=dyn_legPouch_base}
    constraintSetup [
     {targetBone {boneName=pelvis}}
     {targetBone {boneName=thigh_l}}
     {targetBone {boneName=deform_knee_out_l}}
     {targetBone {boneName=thigh_twist_01_l}}
     {targetBone {boneName=deform_groin_l}}
    ]
    constraintWeights [1,1,1,1,1]
   }
   n37 = "Constraint - dyn_scarf_fr_base (PPPPPPPP)"!(
    ComponentPose=@n36
    Alpha=1
    ConstraintWeights_0=0.0269
    ConstraintWeights_1:float=0.0825
    ConstraintWeights_2:float=0.1804
    ConstraintWeights_3:float=0.1347
    ConstraintWeights_4:float=0.0356
    ConstraintWeights_5:float=0.1106
    ConstraintWeights_6:float=0.4046
    ConstraintWeights_7:float=0.0248
   ) -> (Pose) {
    boneToModify {boneName=dyn_scarf_fr_base}
    constraintSetup [
     {targetBone {boneName=spine_03}}
     {targetBone {boneName=spine_04}}
     {targetBone {boneName=spine_05}}
     {targetBone {boneName=clavicle_r}}
     {targetBone {boneName=deform_upperarm_fr_r}}
     {targetBone {boneName=deform_pec_r}}
     {targetBone {boneName=dyn_neck_fr}}
     {targetBone {boneName=dyn_neck_r}}
    ]
    constraintWeights [1,1,1,1,1,1,1,1]
   }
   n38 = "Constraint - dyn_scarf_bk_base (PPPPPPPP)"!(
    ComponentPose=@n37
    Alpha=1
    ConstraintWeights_0=0.1474
    ConstraintWeights_1:float=0.0957
    ConstraintWeights_2:float=0.2929
    ConstraintWeights_3:float=0.3124
    ConstraintWeights_4:float=0.0384
    ConstraintWeights_5:float=0.0469
    ConstraintWeights_6:float=0.0227
    ConstraintWeights_7:float=0.0436
   ) -> (Pose) {
    boneToModify {boneName=dyn_scarf_bk_base}
    constraintSetup [
     {targetBone {boneName=spine_03}}
     {targetBone {boneName=spine_04}}
     {targetBone {boneName=spine_05}}
     {targetBone {boneName=clavicle_l}}
     {targetBone {boneName=deform_upperarm_bk_l}}
     {targetBone {boneName=deform_upperarm_up_l}}
     {targetBone {boneName=dyn_neck_l}}
     {targetBone {boneName=dyn_neck_bk}}
    ]
    constraintWeights [1,1,1,1,1,1,1,1]
   }
   Save cached pose 'Pre phys'(Pose=Component To Local(ComponentPose=@n38)) {cacheName=Pre phys}
   Save cached pose 'Base'(Pose=Copy Pose From Mesh() {bUseAttachedParent=true}) {cacheName=Base}
   Spring controller - Bone: dyn_belly_m_base!(ComponentPose:Component Space Pose Link Structure, Alpha:float=1) -> (Pose:Component Space Pose Link Structure) {is=title AnimGraphNode_SpringBone, springDamping=16, springStiffness=300, springBone {boneName=dyn_belly_m_base}}
  ]}
 EventGraph {nodes [
   Event BlueprintUpdateAnimation()
   > n1 = Sequence!(execute) -> (then_0, then_1, then_2:Exec, then_3:Exec)
   n2 = Break Fort Anim Input Slope Info!(FortAnimInput_SlopeInfo:"Fort Anim Input Slope Info Structure (by ref)"=Get SlopeInfoInput() -> (SlopeInfoInput:Fort Anim Input Slope Info Structure)) -> (RootSlopeTranslation:Vector=~, LeftFootSlopeTranslation:Vector=~, RightFootSlopeTranslation:Vector=~, SlopeRotation:Rotator=~) {is=title K2Node_BreakStruct}
   > Set slope_foot_l_trans(slope_foot_l_trans:Vector=@n2.LeftFootSlopeTranslation) -> (Output_Get:Vector)
   > Set slope_foot_r_trans(slope_foot_r_trans:Vector=@n2.RightFootSlopeTranslation) -> (Output_Get:Vector)
   > Set slope_foot_rot(slope_foot_rot:Rotator=@n2.SlopeRotation) -> (Output_Get:Rotator)
   n3 = Get X Factor() -> (X Factor:double)
   n4 = Rotation adjustment!(
    execute=@n1.then_3
    self
    X Sign:double=-1
    X factor:double=@n3
    Y Sign:double=-1
    Y factor:double=@n3
    Z Sign:double=-1
    Z factor:double=@n3
    Bone To Limit:Name=head
    Reference Bone:Name=neck_02
   ) -> (then, Limited Rotation:Rotator)
   > Set headRotLimited(headRotLimited:Rotator=@n4) -> (Output_Get:Rotator)
   > n5 = Rotation adjustment(X Sign=-1, X factor=@n3, Y Sign=-1, Y factor=@n3, Z Sign=-1, Z factor=@n3, Bone To Limit=neck_02, Reference Bone=neck_01)
   > Set Neck_02 Rot Limited(Neck_02 Rot Limited:Rotator=@n5) -> (Output_Get:Rotator)
   > n6 = Rotation adjustment(X Sign=-1, X factor=@n3, Y Sign=-1, Y factor=@n3, Z Sign=-1, Z factor=@n3, Bone To Limit=neck_01, Reference Bone=spine_05)
   > Set Neck_01 Rot Limited(Neck_01 Rot Limited:Rotator=@n6) -> (Output_Get:Rotator)
   n7 = Get As Fort Player Pawn Athena() -> (As Fort Player Pawn Athena:Fort Player Pawn Athena Object Reference)
   Is Valid(exec=@n1.then_2, InputObject=@n7)
   n8 = GetDisplayName(Object=GetSkeletalMeshForPartType(self=@n7, PartType=Gameplay))
   > Set GauntlesEquip(GauntlesEquip:bool=OR Boolean!(
     A=Contains(SearchIn=@n8, Substring=GauntletsTemplate_Gadget)
     B=Contains(SearchIn=@n8, Substring=SprocketPoppy_Gauntlet)
     C:bool=Contains(SearchIn=@n8, Substring=CoffeePuff_Gauntlet)
     D:bool=Get bIsLeftArmHidden() -> (bIsLeftArmHidden:bool)
     E:bool=Get bIsRightArmHidden() -> (bIsRightArmHidden:bool)
    ) -> (ReturnValue)) -> (Output_Get:bool)
   n9 = GetSkeletalMeshForPartType(self=@n7, PartType=ExtraPart)
   > Set CustomBootsEquip(CustomBootsEquip:bool=Contains(SearchIn=GetDisplayName(Object=@n9), Substring=BootsTemplate_Gadget)) -> (Output_Get:bool)
   n10 = GetSkeletalMeshForPartType(self=Get As Fort Player Pawn Athena())
   Set L_Flop_Alpha(
    execute=@n1.then_1
    L_Flop_Alpha:float=Get Distance Between Two Sockets!(
     Component:Skeletal Mesh Component Object Reference=@n10
     SocketOrBoneNameA:Name=dyn_slope_foot_l
     SocketSpaceA:ERelativeTransformSpace Enum=RTS_World
     SocketOrBoneNameB:Name=ik_foot_l
     SocketSpaceB:ERelativeTransformSpace Enum=RTS_World
     bRemapRange:bool=true
     InRangeMin:float=10
     InRangeMax:float=15
     OutRangeMin:float=0
     OutRangeMax:float=1
    ) -> (ReturnValue:float) {is=title K2Node_CallFunction}
   ) -> (Output_Get:float)
   > Set R_Flop_Alpha(R_Flop_Alpha:float=Get Distance Between Two Sockets(
     Component=@n10
     SocketOrBoneNameA=dyn_slope_foot_r
     SocketSpaceA=RTS_World
     SocketOrBoneNameB=ik_foot_r
     SocketSpaceB=RTS_World
     bRemapRange=true
     InRangeMin=10
     InRangeMax=15
     OutRangeMin=0
     OutRangeMax=1
    )) -> (Output_Get:float)
   > Set LeftArmScarfHide(LeftArmScarfHide:float=GetCurveValue!(self:Skeletal Mesh Component Object Reference=@n10, CurveName=LeftArmScarfHide, DefaultValue:float=0) -> (Value:float, ReturnValue:bool=false)) -> (Output_Get:float)
   Event OnCharacterCustomizationLoaded()
   > n11 = Sequence!(execute) -> (then_0, then_1)
   n12 = TryGetPawnOwner()
   > Is Valid(InputObject=@n12)
   > n13 = Cast To FortPlayerPawnAthena(Object=@n12)
   > Set As Fort Player Pawn Athena(As Fort Player Pawn Athena:Fort Player Pawn Athena Object Reference=@n13) -> (Output_Get:Fort Player Pawn Athena Object Reference)
   Set isWearingBackbling(execute=@n11.then_1, isWearingBackbling=CheckForBackbling())
  ]}
 Rotation adjustment {nodes [
   n1 = Rotation adjustment!() -> (
    then
    X Sign:double=~
    X factor:double=~
    Y Sign:double=~
    Y factor:double=~
    Z Sign:double=~
    Z factor:double=~
    Bone To Limit:Name
    Reference Bone:Name
   ) {is=title K2Node_FunctionEntry}
   n2 = GetOwnerSkeletalMeshComponent()
   n3 = BreakRotator(InRot=BreakTransform(InTransform=GetDeltaTransformFromRefPose(self=@n2, BoneName=@n1.Bone To Limit, BaseName=@n1.Reference Bone)).Rotation)
   > Return Node!(
    execute
    Limited Rotation:Rotator=MakeRotator(
     Roll=float * float!(A=@n3, B=@n1, C:double=@n1.X factor) -> (ReturnValue)
     Pitch=float * float(A=@n3.Pitch, B=@n1.Y Sign, C=@n1.Y factor)
     Yaw=float * float(A=@n3.Yaw, B=@n1.Z Sign, C=@n1.Z factor)
    )
   )
  ]}
]
variables [
 As Fort Player Pawn Athena {type=FortPlayerPawnAthena}
 Backpack {type=CustomCharacterPart, default=null}
 CustomBootsEquip {type=boolean}
 FronEndNotEmote {type=boolean}
 GauntlesEquip {type=boolean}
 headRotLimited {type=Rotator}
 Is Right Arm Hidden {type=boolean}
 isWearingBackbling {type=boolean}
 L_Flop_Alpha {type=number}
 LeftArmScarfHide {type=number}
 Neck_01 Rot Limited {type=Rotator}
 Neck_02 Rot Limited {type=Rotator}
 R_Flop_Alpha {type=number}
 slope_foot_l_trans {type=Vector}
 slope_foot_r_trans {type=Vector}
 slope_foot_rot {type=Rotator}
 X Factor {type=number, default=0.8}
]
