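# Post-style face accessory AnimBP for the ChubbyJingle large male cosmetic body, driving its glasses, whiskers and hair.
# It copies the attached parent's pose, then weights eight whisker bases to facial bones, each followed by spring-based AnimDynamics relative to the head.
# Glasses follow brow and head constraints plus an additive offset, then seven hair bones get spring dynamics.
# The glasses AnimDynamics node is unconnected, and the event graph has no logic.
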AnimBlueprint /BRCosmetics/Characters/Player/Male/Large/Bodies/M_LRG_ChubbyJingle/Meshes/Parts/M_LRG_ChubbyJingle_FaceAcc_AnimBP
target_skeleton=&/M_LRG_ChubbyJingle_FaceAcc_Skeleton
graphs [
 AnimGraph {nodes [
   Anim Dynamics - Bone: dyn_glasses(Alpha=1, ExternalForce="0,0,0") {
    angularDampingOverride=1
    angularSpringConstant=1000
    bAngularSpring=true
    bOverrideAngularDamping=true
    bUseGravityOverride=true
    simulationSpace=BoneRelative
    boundBone {boneName=dyn_glasses}
    componentAppliedLinearAccClamp=<5000 5000 5000>
    componentLinearAccScale=<0.5 0.5 0.5>
    componentLinearVelScale=<0.5 0.5 0.5>
    gravityOverride {z=-980}
    physicsBodyDefinitions [{
      boundBone {boneName=dyn_glasses}
      boxExtents=<10 10 10>
      constraintSetup {angularLimitsMax=<0 0 0>, angularLimitsMin=<0 -10 0>, angularTarget=<1 0 0>}
      localJointOffset=<0.2 0 0>
     }]
    relativeSpaceBone {boneName=root}
   }
   n1 = "Constraint - dyn_whisker_a_l_base (PPPPPP)"!(
    ComponentPose=Local To Component(LocalPose=Copy Pose From Mesh() {bUseAttachedParent=true})
    Alpha=1
    ConstraintWeights_0=0.3832
    ConstraintWeights_1:float=0.0065
    ConstraintWeights_2:float=0.0003
    ConstraintWeights_3:float=0.525
    ConstraintWeights_4:float=0.0001
    ConstraintWeights_5:float=0.0848
   ) -> (Pose) {
    boneToModify {boneName=dyn_whisker_a_l_base}
    constraintSetup [
     {targetBone {boneName=C_nose_bridge}}
     {targetBone {boneName=L_eye_lid_lower_mid}}
     {targetBone {boneName=L_eye_lid_upper_mid}}
     {targetBone {boneName=head}}
     {targetBone {boneName=L_lip_corner}}
     {targetBone {boneName=L_cheek_inner}}
    ]
    constraintWeights [1,1,1,1,1,1]
   }
   n2 = Anim Dynamics - Bone: dyn_whisker_a_l(ComponentPose=@n1, Alpha=1, ExternalForce="0,0,0") {
    angularDampingOverride=1
    angularSpringConstant=1500
    bAngularSpring=true
    bOverrideAngularDamping=true
    bUseGravityOverride=true
    boundBone {boneName=dyn_whisker_a_l}
    gravityOverride {z=-980}
    physicsBodyDefinitions [{
      boundBone {boneName=dyn_whisker_a_l}
      boxExtents=<10 10 10>
      constraintSetup {angularLimitsMax=<3 10 10>, angularLimitsMin=<-3 -10 -10>, angularTarget=<1 0 0>}
      localJointOffset=<0.1 0 0>
     }]
    relativeSpaceBone {boneName=head}
   }
   n3 = "Constraint - dyn_whisker_b_l_base (PPPPPPPP)"!(
    ComponentPose=@n2
    Alpha=1
    ConstraintWeights_0=0.5308
    ConstraintWeights_1:float=0.0706
    ConstraintWeights_2:float=0.0057
    ConstraintWeights_3:float=0.0025
    ConstraintWeights_4:float=0.3889
    ConstraintWeights_5:float=0.0006
    ConstraintWeights_6:float=0.0006
    ConstraintWeights_7:float=0.0003
   ) -> (Pose) {
    boneToModify {boneName=dyn_whisker_b_l_base}
    constraintSetup [
     {targetBone {boneName=head}}
     {targetBone {boneName=L_cheek_inner}}
     {targetBone {boneName=L_lip_upper_outer}}
     {targetBone {boneName=L_lip_corner}}
     {targetBone {boneName=C_nose_bridge}}
     {targetBone {boneName=R_cheek_inner}}
     {targetBone {boneName=C_jaw}}
     {targetBone {boneName=L_lip_lower_outer}}
    ]
    constraintWeights [1,1,1,1,1,1,1,1]
   }
   n4 = Anim Dynamics - Bone: dyn_whisker_b_l(ComponentPose=@n3, Alpha=1, ExternalForce="0,0,0") {
    angularDampingOverride=1
    angularSpringConstant=1500
    bAngularSpring=true
    bOverrideAngularDamping=true
    bUseGravityOverride=true
    boundBone {boneName=dyn_whisker_b_l}
    gravityOverride {z=-980}
    physicsBodyDefinitions [{
      boundBone {boneName=dyn_whisker_b_l}
      boxExtents=<10 10 10>
      constraintSetup {angularLimitsMax=<3 10 10>, angularLimitsMin=<-3 -10 -10>, angularTarget=<1 0 0>}
      localJointOffset=<0.1 0 0>
     }]
    relativeSpaceBone {boneName=head}
   }
   n5 = "Constraint - dyn_whisker_c_l_base (PPPPPP)"!(
    ComponentPose=@n4
    Alpha=1
    ConstraintWeights_0=0.5401
    ConstraintWeights_1:float=0.0398
    ConstraintWeights_2:float=0.0186
    ConstraintWeights_3:float=0.0009
    ConstraintWeights_4:float=0.3999
    ConstraintWeights_5:float=0.0007
   ) -> (Pose) {
    boneToModify {boneName=dyn_whisker_c_l_base}
    constraintSetup [
     {targetBone {boneName=head}}
     {targetBone {boneName=L_cheek_inner}}
     {targetBone {boneName=L_lip_upper_outer}}
     {targetBone {boneName=L_lip_corner}}
     {targetBone {boneName=C_nose_bridge}}
     {targetBone {boneName=R_cheek_inner}}
    ]
    constraintWeights [1,1,1,1,1,1]
   }
   n6 = Anim Dynamics - Bone: dyn_whisker_c_l(ComponentPose=@n5, Alpha=1, ExternalForce="0,0,0") {
    angularDampingOverride=1
    angularSpringConstant=1500
    bAngularSpring=true
    bOverrideAngularDamping=true
    bUseGravityOverride=true
    boundBone {boneName=dyn_whisker_c_l}
    gravityOverride {z=-980}
    physicsBodyDefinitions [{
      boundBone {boneName=dyn_whisker_c_l}
      boxExtents=<10 10 10>
      constraintSetup {angularLimitsMax=<3 10 10>, angularLimitsMin=<-3 -10 -10>, angularTarget=<1 0 0>}
      localJointOffset=<0.1 0 0>
     }]
    relativeSpaceBone {boneName=head}
   }
   n7 = "Constraint - dyn_whisker_d_l_base (PPPPPPPPP)"!(
    ComponentPose=@n6
    Alpha=1
    ConstraintWeights_0=0.4033
    ConstraintWeights_1:float=0.0908
    ConstraintWeights_2:float=0.01
    ConstraintWeights_3:float=0.0706
    ConstraintWeights_4:float=0.1194
    ConstraintWeights_5:float=0.2493
    ConstraintWeights_6:float=0.0403
    ConstraintWeights_7:float=0.0156
    ConstraintWeights_8:float=0.0006
   ) -> (Pose) {
    boneToModify {boneName=dyn_whisker_d_l_base}
    constraintSetup [
     {targetBone {boneName=head}}
     {targetBone {boneName=L_cheek_inner}}
     {targetBone {boneName=C_lip_upper_mid}}
     {targetBone {boneName=L_lip_upper_outer}}
     {targetBone {boneName=L_lip_corner}}
     {targetBone {boneName=C_nose_bridge}}
     {targetBone {boneName=C_jaw}}
     {targetBone {boneName=L_lip_lower_outer}}
     {targetBone {boneName=C_lip_lower_mid}}
    ]
    constraintWeights [1,1,1,1,1,1,1,1,1]
   }
   n8 = Anim Dynamics - Bone: dyn_whisker_d_l(ComponentPose=@n7, Alpha=1, ExternalForce="0,0,0") {
    angularDampingOverride=1
    angularSpringConstant=1500
    bAngularSpring=true
    bOverrideAngularDamping=true
    bUseGravityOverride=true
    boundBone {boneName=dyn_whisker_d_l}
    gravityOverride {z=-980}
    physicsBodyDefinitions [{
      boundBone {boneName=dyn_whisker_d_l}
      boxExtents=<10 10 10>
      constraintSetup {angularLimitsMax=<3 10 10>, angularLimitsMin=<-3 -10 -10>, angularTarget=<1 0 0>}
      localJointOffset=<0.1 0 0>
     }]
    relativeSpaceBone {boneName=head}
   }
   n9 = "Constraint - dyn_whisker_a_r_base (PPPPPP)"!(
    ComponentPose=@n8
    Alpha=1
    ConstraintWeights_0=0.5009
    ConstraintWeights_1:float=0.3562
    ConstraintWeights_2:float=0.1202
    ConstraintWeights_3:float=0.0005
    ConstraintWeights_4:float=0.001
    ConstraintWeights_5:float=0.0212
   ) -> (Pose) {
    boneToModify {boneName=dyn_whisker_a_r_base}
    constraintSetup [
     {targetBone {boneName=head}}
     {targetBone {boneName=C_nose_bridge}}
     {targetBone {boneName=R_cheek_inner}}
     {targetBone {boneName=R_lip_corner}}
     {targetBone {boneName=R_eye_lid_upper_mid}}
     {targetBone {boneName=R_eye_lid_lower_mid}}
    ]
    constraintWeights [1,1,1,1,1,1]
   }
   n10 = Anim Dynamics - Bone: dyn_whisker_a_r(ComponentPose=@n9, Alpha=1, ExternalForce="0,0,0") {
    angularDampingOverride=1
    angularSpringConstant=1500
    bAngularSpring=true
    bOverrideAngularDamping=true
    bUseGravityOverride=true
    boundBone {boneName=dyn_whisker_a_r}
    gravityOverride {z=-980}
    physicsBodyDefinitions [{
      boundBone {boneName=dyn_whisker_a_r}
      boxExtents=<10 10 10>
      constraintSetup {angularLimitsMax=<3 10 10>, angularLimitsMin=<-3 -10 -10>, angularTarget=<1 0 0>}
      localJointOffset=<0.1 0 0>
     }]
    relativeSpaceBone {boneName=head}
   }
   n11 = "Constraint - dyn_whisker_b_r_base (PPPPPPPP)"!(
    ComponentPose=@n10
    Alpha=1
    ConstraintWeights_0=0.5319
    ConstraintWeights_1:float=0.0005
    ConstraintWeights_2:float=0.3895
    ConstraintWeights_3:float=0.072
    ConstraintWeights_4:float=0.0022
    ConstraintWeights_5:float=0.0006
    ConstraintWeights_6:float=0.0003
    ConstraintWeights_7:float=0.003
   ) -> (Pose) {
    boneToModify {boneName=dyn_whisker_b_r_base}
    constraintSetup [
     {targetBone {boneName=head}}
     {targetBone {boneName=L_cheek_inner}}
     {targetBone {boneName=C_nose_bridge}}
     {targetBone {boneName=R_cheek_inner}}
     {targetBone {boneName=R_lip_corner}}
     {targetBone {boneName=C_jaw}}
     {targetBone {boneName=R_lip_lower_outer}}
     {targetBone {boneName=R_lip_upper_outer}}
    ]
    constraintWeights [1,1,1,1,1,1,1,1]
   }
   n12 = Anim Dynamics - Bone: dyn_whisker_b_r(ComponentPose=@n11, Alpha=1, ExternalForce="0,0,0") {
    angularDampingOverride=1
    angularSpringConstant=1500
    bAngularSpring=true
    bOverrideAngularDamping=true
    bUseGravityOverride=true
    boundBone {boneName=dyn_whisker_b_r}
    gravityOverride {z=-980}
    physicsBodyDefinitions [{
      boundBone {boneName=dyn_whisker_b_r}
      boxExtents=<10 10 10>
      constraintSetup {angularLimitsMax=<3 10 10>, angularLimitsMin=<-3 -10 -10>, angularTarget=<1 0 0>}
      localJointOffset=<0.1 0 0>
     }]
    relativeSpaceBone {boneName=head}
   }
   n13 = "Constraint - dyn_whisker_c_r_base (PPPPPPP)"!(
    ComponentPose=@n12
    Alpha=1
    ConstraintWeights_0=0.4972
    ConstraintWeights_1:float=0.388
    ConstraintWeights_2:float=0.0886
    ConstraintWeights_3:float=0.0071
    ConstraintWeights_4:float=0.0017
    ConstraintWeights_5:float=0.0008
    ConstraintWeights_6:float=0.0167
   ) -> (Pose) {
    boneToModify {boneName=dyn_whisker_c_r_base}
    constraintSetup [
     {targetBone {boneName=head}}
     {targetBone {boneName=C_nose_bridge}}
     {targetBone {boneName=R_cheek_inner}}
     {targetBone {boneName=R_lip_corner}}
     {targetBone {boneName=C_jaw}}
     {targetBone {boneName=R_lip_lower_outer}}
     {targetBone {boneName=R_lip_upper_outer}}
    ]
    constraintWeights [1,1,1,1,1,1,1]
   }
   n14 = Anim Dynamics - Bone: dyn_whisker_c_r(ComponentPose=@n13, Alpha=1, ExternalForce="0,0,0") {
    angularDampingOverride=1
    angularSpringConstant=1500
    bAngularSpring=true
    bOverrideAngularDamping=true
    bUseGravityOverride=true
    boundBone {boneName=dyn_whisker_c_r}
    gravityOverride {z=-980}
    physicsBodyDefinitions [{
      boundBone {boneName=dyn_whisker_c_r}
      boxExtents=<10 10 10>
      constraintSetup {angularLimitsMax=<3 10 10>, angularLimitsMin=<-3 -10 -10>, angularTarget=<1 0 0>}
      localJointOffset=<0.1 0 0>
     }]
    relativeSpaceBone {boneName=head}
   }
   n15 = "Constraint - dyn_whisker_d_r_base (PPPPPPPPP)"!(
    ComponentPose=@n14
    Alpha=1
    ConstraintWeights_0=0.4752
    ConstraintWeights_1:float=0
    ConstraintWeights_2:float=0.0121
    ConstraintWeights_3:float=0.3461
    ConstraintWeights_4:float=0.0666
    ConstraintWeights_5:float=0.036
    ConstraintWeights_6:float=0.0115
    ConstraintWeights_7:float=0.004
    ConstraintWeights_8:float=0.0485
   ) -> (Pose) {
    boneToModify {boneName=dyn_whisker_d_r_base}
    constraintSetup [
     {targetBone {boneName=head}}
     {targetBone {boneName=L_cheek_inner}}
     {targetBone {boneName=C_lip_upper_mid}}
     {targetBone {boneName=C_nose_bridge}}
     {targetBone {boneName=R_cheek_inner}}
     {targetBone {boneName=R_lip_corner}}
     {targetBone {boneName=C_jaw}}
     {targetBone {boneName=R_lip_lower_outer}}
     {targetBone {boneName=R_lip_upper_outer}}
    ]
    constraintWeights [1,1,1,1,1,1,1,1,1]
   }
   n16 = Anim Dynamics - Bone: dyn_whisker_d_r(ComponentPose=@n15, Alpha=1, ExternalForce="0,0,0") {
    angularDampingOverride=1
    angularSpringConstant=1500
    bAngularSpring=true
    bOverrideAngularDamping=true
    bUseGravityOverride=true
    boundBone {boneName=dyn_whisker_d_r}
    gravityOverride {z=-980}
    physicsBodyDefinitions [{
      boundBone {boneName=dyn_whisker_d_r}
      boxExtents=<10 10 10>
      constraintSetup {angularLimitsMax=<3 10 10>, angularLimitsMin=<-3 -10 -10>, angularTarget=<1 0 0>}
      localJointOffset=<0.1 0 0>
     }]
    relativeSpaceBone {boneName=head}
   }
   n17 = "Constraint - dyn_glasses (PPPP)"!(
    ComponentPose=@n16
    Alpha=1
    ConstraintWeights_0=0.2
    ConstraintWeights_1:float=0.1
    ConstraintWeights_2:float=0.1
    ConstraintWeights_3:float=0.6
   ) -> (Pose) {
    boneToModify {boneName=dyn_glasses}
    constraintSetup [
     {perAxis {bZ=false}, targetBone {boneName=C_brow_mid}}
     {perAxis {bZ=false}, targetBone {boneName=L_brow_mid}}
     {perAxis {bZ=false}, targetBone {boneName=R_brow_mid}}
     {perAxis {bZ=false}, targetBone {boneName=head}}
    ]
    constraintWeights [1,1,1,1]
   }
   n18 = "Transform (Modify) Bone - Bone: dyn_glasses"(ComponentPose=@n17, Alpha=1, Translation="0,2,0", Rotation="0,0,-4", Scale="1,1,1") {
    rotationMode=BMM_Additive
    rotationSpace=BCS_ParentBoneSpace
    translationMode=BMM_Additive
    boneToModify {boneName=dyn_glasses}
    rotation {roll=-4}
    translation {y=2}
   }
   n19 = Anim Dynamics - Bone: dyn_hair_a(ComponentPose=@n18, Alpha=1, ExternalForce="0,0,0") {
    angularDampingOverride=1
    angularSpringConstant=1500
    bAngularSpring=true
    bOverrideAngularDamping=true
    bUseGravityOverride=true
    boundBone {boneName=dyn_hair_a}
    componentAppliedLinearAccClamp=<5000 5000 5000>
    componentLinearAccScale=<0.5 0.5 0.5>
    componentLinearVelScale=<0.5 0.5 0.5>
    gravityOverride {z=-980}
    physicsBodyDefinitions [{
      boundBone {boneName=dyn_hair_a}
      boxExtents=<10 10 10>
      constraintSetup {angularLimitsMax=<3 15 15>, angularLimitsMin=<-3 -15 -15>, angularTarget=<1 0 0>}
      localJointOffset=<0.1 0 0>
     }]
    relativeSpaceBone {boneName=root}
   }
   n20 = Anim Dynamics - Bone: dyn_hair_b(ComponentPose=@n19, Alpha=1, ExternalForce="0,0,0") {
    angularDampingOverride=1
    angularSpringConstant=1500
    bAngularSpring=true
    bOverrideAngularDamping=true
    bUseGravityOverride=true
    boundBone {boneName=dyn_hair_b}
    componentAppliedLinearAccClamp=<5000 5000 5000>
    componentLinearAccScale=<0.5 0.5 0.5>
    componentLinearVelScale=<0.5 0.5 0.5>
    gravityOverride {z=-980}
    physicsBodyDefinitions [{
      boundBone {boneName=dyn_hair_b}
      boxExtents=<15 15 15>
      constraintSetup {angularLimitsMax=<3 15 15>, angularLimitsMin=<-3 -15 -15>, angularTarget=<1 0 0>}
      localJointOffset=<0.1 0 0>
     }]
    relativeSpaceBone {boneName=root}
   }
   n21 = Anim Dynamics - Bone: dyn_hair_c(ComponentPose=@n20, Alpha=1, ExternalForce="0,0,0") {
    angularDampingOverride=1
    angularSpringConstant=1500
    bAngularSpring=true
    bOverrideAngularDamping=true
    bUseGravityOverride=true
    boundBone {boneName=dyn_hair_c}
    componentAppliedLinearAccClamp=<5000 5000 5000>
    componentLinearAccScale=<0.5 0.5 0.5>
    componentLinearVelScale=<0.5 0.5 0.5>
    gravityOverride {z=-980}
    physicsBodyDefinitions [{
      boundBone {boneName=dyn_hair_c}
      boxExtents=<12 12 12>
      constraintSetup {angularLimitsMax=<3 15 15>, angularLimitsMin=<-3 -15 -15>, angularTarget=<1 0 0>}
      localJointOffset=<0.1 0 0>
     }]
    relativeSpaceBone {boneName=root}
   }
   n22 = Anim Dynamics - Bone: dyn_hair_d(ComponentPose=@n21, Alpha=1, ExternalForce="0,0,0") {
    angularDampingOverride=1
    angularSpringConstant=1500
    bAngularSpring=true
    bOverrideAngularDamping=true
    bUseGravityOverride=true
    boundBone {boneName=dyn_hair_d}
    componentAppliedLinearAccClamp=<5000 5000 5000>
    componentLinearAccScale=<0.5 0.5 0.5>
    componentLinearVelScale=<0.5 0.5 0.5>
    gravityOverride {z=-980}
    physicsBodyDefinitions [{
      boundBone {boneName=dyn_hair_d}
      boxExtents=<16 16 16>
      constraintSetup {angularLimitsMax=<3 15 15>, angularLimitsMin=<-3 -15 -15>, angularTarget=<1 0 0>}
      localJointOffset=<0.1 0 0>
     }]
    relativeSpaceBone {boneName=root}
   }
   n23 = Anim Dynamics - Bone: dyn_hair_e(ComponentPose=@n22, Alpha=1, ExternalForce="0,0,0") {
    angularDampingOverride=1
    angularSpringConstant=1500
    bAngularSpring=true
    bOverrideAngularDamping=true
    bUseGravityOverride=true
    boundBone {boneName=dyn_hair_e}
    componentAppliedLinearAccClamp=<5000 5000 5000>
    componentLinearAccScale=<0.5 0.5 0.5>
    componentLinearVelScale=<0.5 0.5 0.5>
    gravityOverride {z=-980}
    physicsBodyDefinitions [{
      boundBone {boneName=dyn_hair_e}
      boxExtents=<14 14 14>
      constraintSetup {angularLimitsMax=<3 15 15>, angularLimitsMin=<-3 -15 -15>, angularTarget=<1 0 0>}
      localJointOffset=<0.1 0 0>
     }]
    relativeSpaceBone {boneName=root}
   }
   n24 = Anim Dynamics - Bone: dyn_hair_f(ComponentPose=@n23, Alpha=1, ExternalForce="0,0,0") {
    angularDampingOverride=1
    angularSpringConstant=1500
    bAngularSpring=true
    bOverrideAngularDamping=true
    bUseGravityOverride=true
    boundBone {boneName=dyn_hair_f}
    componentAppliedLinearAccClamp=<5000 5000 5000>
    componentLinearAccScale=<0.5 0.5 0.5>
    componentLinearVelScale=<0.5 0.5 0.5>
    gravityOverride {z=-980}
    physicsBodyDefinitions [{
      boundBone {boneName=dyn_hair_f}
      boxExtents=<13 13 13>
      constraintSetup {angularLimitsMax=<3 15 15>, angularLimitsMin=<-3 -15 -15>, angularTarget=<1 0 0>}
      localJointOffset=<0.1 0 0>
     }]
    relativeSpaceBone {boneName=root}
   }
   n25 = Anim Dynamics - Bone: dyn_hair_g(ComponentPose=@n24, Alpha=1, ExternalForce="0,0,0") {
    angularDampingOverride=1
    angularSpringConstant=1500
    bAngularSpring=true
    bOverrideAngularDamping=true
    bUseGravityOverride=true
    boundBone {boneName=dyn_hair_g}
    componentAppliedLinearAccClamp=<5000 5000 5000>
    componentLinearAccScale=<0.5 0.5 0.5>
    componentLinearVelScale=<0.5 0.5 0.5>
    gravityOverride {z=-980}
    physicsBodyDefinitions [{
      boundBone {boneName=dyn_hair_g}
      boxExtents=<11 11 11>
      constraintSetup {angularLimitsMax=<3 15 15>, angularLimitsMin=<-3 -15 -15>, angularTarget=<1 0 0>}
      localJointOffset=<0.1 0 0>
     }]
    relativeSpaceBone {boneName=root}
   }
   Output Pose(Result=Component To Local(ComponentPose=@n25)) {layerGroup=None}
  ]}
 EventGraph {nodes [Event BlueprintUpdateAnimation()]}
]
variables []
